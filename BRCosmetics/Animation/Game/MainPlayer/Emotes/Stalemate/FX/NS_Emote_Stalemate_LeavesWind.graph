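# Niagara system for the Stalemate emote's wind and leaf effect, with all four emitters currently disabled.
# Leaves_Vortex spawns orange mesh leaves pushed along -Y with curl noise and drag.
# Air_Trail001 adds drifting sprite/ribbon streaks, Straightwind slides mesh/sprite gusts along -Y by direct position offsets, and SwirlWind bursts a delayed mesh swirl.
# The three wind emitters use time-of-day tints, alpha-fade curves and dynamic material parameters.
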NiagaraSystem /BRCosmetics/Animation/Game/MainPlayer/Emotes/Stalemate/FX/NS_Emote_Stalemate_LeavesWind
dependencies {
 usedDataInterfaces [&NiagaraDataInterfaceCurve, &NiagaraDataInterfaceSpriteRendererInfo]
 usedDynamicInputs [&SimulationPosition, &ReturnNormalizedExecIndex]
 usedModules [&SystemState]
}
emitters {
 Air_Trail001 {
  bEnabled=false
  emitterUpdateScript [
   &EmitterState()
   &SpawnRate(Use Spawn Probability=false)
  ]
  particleSpawnScript [
   &InitializeParticle(
    Color=&MultiplyLinearColor(
     Linear Color=&FNTimeOfDayColor(
      Night Color=<rgba 0.5156 0.6758 0.75 0.5>
      Morning Color=<rgba 0.1934 0.2534 0.2812 0.6>
      Day Color=<rgba 0.6875 0.901 1 0.7>
      Evening Color=<rgba 0.3044 0.3989 0.4427 0.6>
     )
     Scale Factor=&MakeFloat4FromVectorAndFloat("Vector (RGB)"=<1 1 1>, "Float (Alpha)"=1)
    )
    Lifetime=&RandomRangeFloat(Minimum=1.8, Maximum=2.4)
    Mass Mode=::NewEnumerator1
    Sprite Size Mode=::NewEnumerator3
    Uniform Sprite Size=&Multiply_Float(A=&RandomRangeFloat(Minimum=16, Maximum=21), B=4)
    UsePositionOffset=false
    Write Material Random=true
   )
   &ShapeLocation(
    Cylinder Height=200
    Cylinder Height Midpoint=1
    Cylinder Radius=40
    MaskSpawnGroup=false
    Offset=<70 100 20>
    Offset Mode=::NewEnumerator0
    Override Seed=false
    Rotation Angle=90
    Rotation Angle Type:ENiagara_AngleInput=::NewEnumerator0
    Rotation Axis=<0 1 0>
    Rotation Coordinate Space=::Simulation
    Rotation Mode=::NewEnumerator1
    Shape Primitive=::NewEnumerator1
    Surface Only=false
    UseOwnerScale=true
   )
   &AddVelocity(
    Cone Angle=0
    Cone Axis=<0 -1 0>
    Module.UseConeFalloff=false
    Module.UseVelDistribution=false
    Override Seed=false
    Velocity Mode=::NewEnumerator2
    Velocity Speed=&Multiply_Float(A=1, B=&RandomRangeFloat(Minimum=75, Maximum=10))
   )
  ]
  particleUpdateScript [
   &ParticleState()
   &ScaleSpriteSize(Scale Factor=<1 3>, Scale Sprite Size Mode=::NewEnumerator2)
   &ScaleColor(
    Scale Alpha=&FloatFromCurve(
     FloatCurve=DataInterface() {propertyValues {Curve {keys [
         {interpMode=RCIM_Cubic, tangentMode=RCTM_User, time=0, value=0}
         {arriveTangent=0.2265, leaveTangent=2, tangentMode=RCTM_User, time=0.3, value=1}
         <at 0.5 1>
         {arriveTangent=0.0955, interpMode=RCIM_Cubic, leaveTangent=0.0955, tangentMode=RCTM_User, time=0.7, value=1}
         {tangentMode=RCTM_User, time=1, value=0}
        ]}}}
     Scale Curve=0.5
    )
    ScaleA=true
    ScaleRGB=false
   )
   &ScaleRibbonWidth() {enabled=false}
   &CurlNoiseForce(Pan Noise=false) {enabled=false}
   &Drag(
    Drag=&Multiply_Float(
     A=&FloatFromCurve(FloatCurve=DataInterface() {propertyValues {Curve {keys [{interpMode=RCIM_Cubic, tangentMode=RCTM_User, time=0, value=0}, {interpMode=RCIM_Cubic, tangentMode=RCTM_User, time=1, value=2}]}}})
     B=&RandomRangeFloat(Minimum=0.8, Maximum=1.2)
    )
    Use Linear Drag=true
    Use Rotational Drag=false
   ) {enabled=false}
   &SolveForcesAndVelocity(Clamp Velocity=false, Limit Acceleration=false)
   &DynamicMaterialParameters(
    Index 0 Param 1=&RandomRangeFloat(Minimum=0, Maximum=1)
    Index 0 Param 2=&RandomRangeFloat(Minimum=0, Maximum=1)
    Module.Index0Param0Write=true
    Module.Index0Param1Write=true
    Module.Index0Param2Write=true
    Module.Index0Param3Write=true
   )
   &SubUVAnimation(
    End Frame Range Override=3
    Evaluation Type=::NewEnumerator0
    "Sprite Renderer "=DataInterface() {propertyValues {SpriteRenderer=&self:Air_Trail001_3.NiagaraSpriteRendererProperties_0}}
    SubUV Animation Mode:ENiagara_SubUVLookupModeV2=::NewEnumerator2
    UseEndFrame=true
    UseStartFrame=true
   ) {enabled=false}
  ]
  rendererClasses [&NiagaraRibbonRendererProperties, &NiagaraSpriteRendererProperties]
  renderers [
   {rendererClass=&NiagaraRibbonRendererProperties}
   {rendererClass=&NiagaraSpriteRendererProperties, rendererIndex=1}
  ]
 }
 Leaves_Vortex {
  bEnabled=false
  emitterUpdateScript [
   &EmitterState()
   &SpawnRate(SpawnRate=4, Use Spawn Probability=false)
  ]
  particleSpawnScript [
   &InitializeParticle(
    Color=<rgba 2.5 0.25 0 1>
    Lifetime Max=2.5
    Lifetime Min=1.2
    Lifetime Mode=::NewEnumerator1
    Mesh Scale Mode=::NewEnumerator1
    Mesh Uniform Scale Max=1.1
    Mesh Uniform Scale Min=0.8
    Sprite Size Mode=::NewEnumerator3
    Uniform Sprite Size=&RandomRangeFloat(Minimum=16, Maximum=21)
    UsePositionOffset=false
    Write Material Random=true
   )
   &ShapeLocation(
    Cylinder Height=200
    Cylinder Height Midpoint=1
    Cylinder Radius=40
    MaskSpawnGroup=false
    Offset=<70 100 20>
    Offset Mode=::NewEnumerator0
    Override Seed=false
    Rotation Angle=90
    Rotation Angle Type:ENiagara_AngleInput=::NewEnumerator0
    Rotation Axis=<0 1 0>
    Rotation Coordinate Space=::Simulation
    Rotation Mode=::NewEnumerator1
    Shape Primitive=::NewEnumerator1
    Surface Only=false
    UseOwnerScale=true
   )
   &InitialMeshOrientation(Use Rotation Vector=false)
   &AddVelocity(
    Cone Axis=<0 -1 0>
    Module.UseConeFalloff=false
    Module.UseVelDistribution=false
    Override Seed=false
    Velocity Mode=::NewEnumerator2
    Velocity Speed=&Multiply_Float(A=2, B=&RandomRangeFloat(Minimum=75, Maximum=35))
   )
   &AddRotationalVelocity(Rotation Rate=&Multiply_VectorByFloat(
     Vector=&RandomRangeVector(Minimum=<0.5 0.5 0.5>, Maximum=<1 1 1>)
     Float=&Multiply_Float(
      A=&UniformAOrBFloat(A=-1, Evaluation Type=::NewEnumerator0, Recalculate Random Each Loop=true, Override Seed=false)
      B=0.25
     )
    ))
  ]
  particleUpdateScript [
   &ParticleState()
   /SetVariables(Particles.Position=&AddVectorToPosition(
     Position=$Particles.Initial.Position
     Vector=&MakeVector(
      X=&Lerp_Float(A=-225, B=200, Alpha=$Particles.NormalizedAge)
      Y=0
      Z=&Add_Float(
       A=&FloatFromCurve(
        FloatCurve=DataInterface() {propertyValues {Curve {keys [
            {arriveTangent=-2.1826, interpMode=RCIM_Cubic, leaveTangent=-2.1826, tangentMode=RCTM_User, time=0, value=1.5}
            {arriveTangent=-0.5, interpMode=RCIM_Cubic, leaveTangent=-0.5, time=0.5, value=0.5}
            {arriveTangent=2.8781, interpMode=RCIM_Cubic, leaveTangent=2.8781, tangentMode=RCTM_User, time=1, value=1}
           ]}}}
        Scale Curve=50
       )
       B=50
      )
     )
    )) {enabled=false, isSetParameters=true}
   &ScaleMeshSize(Scale Factor=&VectorFromFloat(Value=&FloatFromCurve(
      FloatCurve=DataInterface() {propertyValues {Curve {keys [
          {arriveTangent=9.0725, interpMode=RCIM_Cubic, leaveTangent=9.0725, tangentMode=RCTM_User, time=0, value=0}
          {arriveTangent=0.3207, interpMode=RCIM_Cubic, leaveTangent=0.3207, tangentMode=RCTM_User, time=0.1408, value=1}
          {arriveTangent=-3.5468, interpMode=RCIM_Cubic, leaveTangent=-3.5468, tangentMode=RCTM_User, time=1, value=0}
         ]}}}
      Scale Curve=0.8
     )))
   &UpdateMeshOrientation(Rotation Rate=8) {enabled=false, module=UpdateMeshOrientation001}
   &UpdateMeshOrientation(
    Facing Interpolation=false
    Facing Position:Vector3f=$Particles.Velocity
    Orientation Method=::NewEnumerator2
    Override Quat=false
    TransformFacingOrigin=false
    TransformFacingPosition=false
    TransformUp=false
   ) {enabled=false}
   &CurlNoiseForce(Pan Noise=false)
   &Drag(
    Drag=&Multiply_Float(
     A=&FloatFromCurve(FloatCurve=DataInterface() {propertyValues {Curve {keys [{interpMode=RCIM_Cubic, tangentMode=RCTM_User, time=0, value=0}, {interpMode=RCIM_Cubic, tangentMode=RCTM_User, time=1, value=2}]}}})
     B=&RandomRangeFloat(Minimum=0.8, Maximum=1.2)
    )
    Use Linear Drag=true
    Use Rotational Drag=false
   )
   &SolveForcesAndVelocity(Clamp Velocity=false, Limit Acceleration=false)
  ]
  rendererClasses [&NiagaraMeshRendererProperties]
  renderers [{rendererClass=&NiagaraMeshRendererProperties, rendererIndex=0}]
 }
 Straightwind {
  bEnabled=false
  emitterUpdateScript [
   &EmitterState()
   &SpawnBurst_Instantaneous(Use Spawn Probability=false, UseLoopCountLimit=false) {enabled=false}
   &SpawnRate(SpawnRate=3, Use Spawn Probability=false)
  ]
  particleSpawnScript [
   &InitializeParticle(
    Color=&FNTimeOfDayColor(
     Night Color=<rgba 0.4667 0.5464 0.5833 0.4>
     Morning Color=<rgba 0.5958 0.6976 0.7448 0.6>
     Day Color=<rgba 0.8 0.9367 1 0.6>
     Evening Color=<rgba 0.575 0.6732 0.7188 0.6>
    )
    Lifetime=&RandomRangeFloat(Minimum=0.8, Maximum=1.5)
    Mesh Scale Mode=::NewEnumerator3
    Mesh Uniform Scale=2
    UsePositionOffset=false
    Write Material Random=true
   )
   &InitialMeshOrientation(Mesh Axis To Orient=<0 1 0>, Mesh Orientation Mode=::NewEnumerator2, Use Rotation Vector=false)
   &ShapeLocation(
    Cylinder Height=50
    Cylinder Height Midpoint=1
    Cylinder Radius=40
    MaskSpawnGroup=false
    Offset=<80 300 50>
    Offset Mode=::NewEnumerator0
    Override Seed=false
    Shape Primitive=::NewEnumerator1
    Surface Only=false
    UseOwnerScale=true
   )
  ]
  particleUpdateScript [
   &ParticleState()
   &DynamicMaterialParameters(
    Index 0 Param 1=&FloatFromCurve(FloatCurve=DataInterface() {propertyValues {Curve {keys [
         {arriveTangent=2.4535, interpMode=RCIM_Cubic, leaveTangent=2.4535, tangentMode=RCTM_User, time=0, value=0}
         {interpMode=RCIM_Cubic, tangentMode=RCTM_User, time=0.5, value=0.5}
         {arriveTangent=-2.5865, tangentMode=RCTM_User, time=1, value=0}
        ]}}})
    Index 0 Param 3=&RandomRangeFloat(Minimum=0.8, Maximum=1.6)
    Index 1 Param 1=&RandomRangeFloat(Minimum=0, Maximum=1)
    Index 1 Param 2=&RandomRangeFloat(Minimum=0, Maximum=1)
    Module.Index0Param0Write=true
    Module.Index0Param1Write=true
    Module.Index0Param2Write=false
    Module.Index0Param3Write=false
    Module.Index1Param0Write=true
    Module.Index1Param1Write=true
    Module.Index1Param2Write=true
    Module.Index1Param3Write=true
    Write Parameter Index 1=true
   )
   &ScaleColor(
    Scale Alpha=&FloatFromCurve(
     FloatCurve=DataInterface() {propertyValues {Curve {keys [
         {interpMode=RCIM_Cubic, tangentMode=RCTM_User, time=0, value=0}
         {interpMode=RCIM_Cubic, tangentMode=RCTM_User, time=0.5, value=1}
         {tangentMode=RCTM_User, time=1, value=0}
        ]}}}
     Scale Curve=0.1
    )
    ScaleA=true
    ScaleRGB=true
   )
   /SetVariables(Particles.Position=&AddVectorToPosition(Position=$Particles.Position, Vector=&MakeVector(X=0, Y=-4, Z=0))) {isSetParameters=true}
  ]
  rendererClasses [&NiagaraMeshRendererProperties, &NiagaraSpriteRendererProperties]
  renderers [
   {rendererClass=&NiagaraMeshRendererProperties, rendererIndex=0}
   {rendererClass=&NiagaraSpriteRendererProperties, rendererIndex=1}
  ]
 }
 SwirlWind {
  bEnabled=false
  emitterUpdateScript [
   &EmitterState()
   &SpawnBurst_Instantaneous(Spawn Time=1.4, Use Spawn Probability=false, UseLoopCountLimit=false)
   &SpawnRate(SpawnRate=10, Use Spawn Probability=false) {enabled=false}
  ]
  particleSpawnScript [
   &InitializeParticle(
    Color=&FNTimeOfDayColor(
     Night Color=<rgba 0.4667 0.5464 0.5833 0.4>
     Morning Color=<rgba 0.5958 0.6976 0.7448 0.6>
     Day Color=<rgba 0.8 0.9367 1 0.6>
     Evening Color=<rgba 0.575 0.6732 0.7188 0.6>
    )
    Mesh Scale Mode=::NewEnumerator3
    Mesh Uniform Scale=2
    Position Offset=<100 60 15>
    UsePositionOffset=true
    Write Material Random=true
   )
   &InitialMeshOrientation(Mesh Axis To Orient=<0 1 0>, Mesh Orientation Mode=::NewEnumerator2, Use Rotation Vector=false)
  ]
  particleUpdateScript [
   &ParticleState()
   &DynamicMaterialParameters(
    Index 0 Param 1=&FloatFromCurve(FloatCurve=DataInterface() {propertyValues {Curve {keys [
         {arriveTangent=2.4535, interpMode=RCIM_Cubic, leaveTangent=2.4535, tangentMode=RCTM_User, time=0, value=0}
         {interpMode=RCIM_Cubic, tangentMode=RCTM_User, time=0.5, value=0.5}
         {arriveTangent=-2.5865, tangentMode=RCTM_User, time=1, value=0}
        ]}}})
    Index 0 Param 2=&FloatFromCurve(FloatCurve=DataInterface() {propertyValues {Curve {keys [<at 0 1>,<at 0.5 0>,<at 0.6 0>,<at 1 -1>]}}})
    Module.Index0Param0Write=true
    Module.Index0Param1Write=true
    Module.Index0Param2Write=false
    Module.Index0Param3Write=false
   )
   &ScaleColor(
    Scale Alpha=&FloatFromCurve(
     FloatCurve=DataInterface() {propertyValues {Curve {keys [
         {interpMode=RCIM_Cubic, tangentMode=RCTM_User, time=0, value=0}
         {arriveTangent=0.0217, interpMode=RCIM_Cubic, leaveTangent=0.0217, tangentMode=RCTM_User, time=0.15, value=1}
         {interpMode=RCIM_Cubic, tangentMode=RCTM_User, time=0.85, value=1}
         {tangentMode=RCTM_User, time=1, value=0}
        ]}}}
     Scale Curve=0.2
    )
    ScaleA=true
    ScaleRGB=true
   )
  ]
  rendererClasses [&NiagaraMeshRendererProperties]
  renderers [{rendererClass=&NiagaraMeshRendererProperties, rendererIndex=0}]
 }
}
summary {
 systemName=NS_Emote_Stalemate_LeavesWind
 emitters [
  {bEnabled=false, emitterName=Leaves_Vortex, rendererClasses [&NiagaraMeshRendererProperties]}
  {
   bEnabled=false
   emitterName=Air_Trail001
   rendererClasses [&NiagaraRibbonRendererProperties, &NiagaraSpriteRendererProperties]
  }
  {bEnabled=false, emitterName=SwirlWind, rendererClasses [&NiagaraMeshRendererProperties]}
  {
   bEnabled=false
   emitterName=Straightwind
   rendererClasses [&NiagaraMeshRendererProperties, &NiagaraSpriteRendererProperties]
  }
 ]
 userVariables [
  User.bIsFrontEnd {type=NiagaraBool}
  User.bIsFrontEndPreview {type=NiagaraBool}
 ]
}
