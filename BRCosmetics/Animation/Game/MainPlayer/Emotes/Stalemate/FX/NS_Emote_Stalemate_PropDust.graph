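# Niagara dust effect for the hubcap gadget prop in the Stalemate emote, continuously emitting soft sprites despite its DirectionalBurst emitter name.
# Particles spawn in a small ring around the prop's JNT_Main_Wheel bone via skeletal mesh sampling, then drift along a cone, slowed by drag.
# Curves grow and fade them, while randomized, age-driven dynamic material parameters vary the look across two sprite renderers.
# An AlphaScale user parameter is declared.
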NiagaraSystem /BRCosmetics/Animation/Game/MainPlayer/Emotes/Stalemate/FX/NS_Emote_Stalemate_PropDust
dependencies {
 usedDataInterfaces [&NiagaraDataInterfaceSkeletalMesh, &NiagaraDataInterfaceCurve]
 usedDynamicInputs []
 usedModules [&SystemState]
}
emitters {DirectionalBurst {
  emitterUpdateScript [
   &EmitterState()
   &SpawnRate(SpawnRate=8, Use Spawn Probability=false)
  ]
  particleSpawnScript [
   &InitializeParticle(
    Color=<rgba 1 0.7803 0.65 0.5>
    Lifetime Min=0.9
    Lifetime Mode=::NewEnumerator1
    Mass Max=2
    Mass Mode=::NewEnumerator3
    Sprite Rotation Mode=::NewEnumerator2
    Sprite Size Mode=::NewEnumerator3
    Uniform Sprite Size=&Multiply_Float(A=6, B=&RandomRangeFloat(Minimum=9, Maximum=13))
    UsePositionOffset=false
    Write Material Random=true
   )
   &SkeletalMeshLocation(
    Bone Sampling Mode=::NewEnumerator4
    MaskSpawnGroup=false
    Skeletal Mesh=DataInterface() {propertyValues {
      DefaultMesh=null
      ExcludeBoneName=root
      OptionalComponentClass=null
      PreviewMesh=/BRCosmetics/Animation/Props/Gadgets_Stalemate_Hubcap/Meshes/Stalemate_Hubcap_Gadget
      SkinningMode=SkinOnTheFly
      SoftSourceActor=null
      SourceMode=Default
      UvSetIndex=0
      WholeMeshLOD=-1
      bExcludeBone=false
      bReadDeformedGeometry=true
      bRequireCurrentFrameData=true
      ComponentTags []
      FilteredBones [JNT_Main_Wheel]
      FilteredSockets []
      MeshUserParameter {parameter {name=None, typeDefHandle {registeredTypeIndex=1}, varData []}}
      SamplingRegions []
     }}
   )
   &ShapeLocation(
    Disc Coverage=1
    MaskSpawnGroup=false
    Offset Mode=::NewEnumerator0
    Override Seed=false
    Ring Radius=5
    Rotation Angle=90
    Rotation Angle Type:ENiagara_AngleInput=::NewEnumerator0
    Rotation Axis=<0 0 0>
    Rotation Coordinate Space=::World
    Rotation Mode=::NewEnumerator1
    Shape Primitive=::NewEnumerator4
    UseOwnerScale=true
   )
   &AddVelocity(
    Cone Axis=<-1 -1.5 1>
    Distribution Along Cone Axis=0.2
    Module.UseConeFalloff=false
    Module.UseVelDistribution=true
    Override Seed=false
    Velocity Mode=::NewEnumerator2
    Velocity Speed=&Multiply_Float(A=&RandomRangeFloat(Minimum=25, Maximum=75), B=0.25)
   )
  ]
  particleUpdateScript [
   &ParticleState()
   &ScaleSpriteSize(
    Uniform Curve Scale=0.8
    Uniform Curve Sprite Scale=DataInterface() {propertyValues {Curve {keys [{arriveTangent=1.1766, interpMode=RCIM_Cubic, leaveTangent=1.1766, tangentMode=RCTM_User, time=0, value=0.25}, <at 1 1 RCIM_Cubic>]}}}
   )
   &Drag(Drag=&RandomRangeFloat(Minimum=0.8, Maximum=1.2), Use Linear Drag=true, Use Rotational Drag=false)
   &ScaleColor(
    Scale Alpha=&FloatFromCurve(
     FloatCurve=DataInterface() {propertyValues {Curve {keys [
         {arriveTangent=9.0725, interpMode=RCIM_Cubic, leaveTangent=9.0725, tangentMode=RCTM_User, time=0, value=0}
         {arriveTangent=0.2816, interpMode=RCIM_Cubic, leaveTangent=0.2816, tangentMode=RCTM_User, time=0.2, value=0.9847}
         {arriveTangent=-1.5192, interpMode=RCIM_Cubic, leaveTangent=-1.5192, tangentMode=RCTM_User, time=0.8, value=0.5}
         {arriveTangent=-0.0405, interpMode=RCIM_Cubic, leaveTangent=-0.0405, tangentMode=RCTM_User, time=1, value=0}
        ]}}}
     Scale Curve=0.8
    )
    Scale RGB=&VectorFromFloat(Value=0.25)
    ScaleA=true
    ScaleRGB=true
   )
   &SolveForcesAndVelocity(Clamp Velocity=false, Limit Acceleration=false)
   &DynamicMaterialParameters(
    Index 0 Param 1=&Add_Float(
     A=&RandomRangeFloat(Minimum=0, Maximum=1)
     B=&FloatFromCurve(
      FloatCurve=DataInterface() {propertyValues {Curve {keys [
          <at 0 0 RCIM_Cubic>
          {arriveTangent=1.5947, interpMode=RCIM_Cubic, leaveTangent=1.5947, tangentMode=RCTM_User, time=0.2557, value=0.5257}
          <at 1 1 RCIM_Cubic>
         ]}}}
      CurveIndex=$Particles.Age
      Scale Curve=0.25
     )
    )
    Index 2 Param 1=&FloatFromCurve(FloatCurve=DataInterface() {propertyValues {Curve {keys [{interpMode=RCIM_Cubic, tangentMode=RCTM_User, time=0.25, value=0}, {interpMode=RCIM_Cubic, tangentMode=RCTM_User, time=0.9, value=1}]}}})
    Index 2 Param 2=&RandomRangeFloat(Minimum=0, Maximum=1)
    Index 2 Param 3=&Add_Float(
     A=1
     B=&FloatFromCurve(FloatCurve=DataInterface() {propertyValues {Curve {keys [{tangentMode=RCTM_User, time=0, value=0}, {tangentMode=RCTM_User, time=1, value=1}]}}})
    )
    Index 2 Param 4=1.5
    Module.Index0Param0Write=true
    Module.Index0Param1Write=true
    Module.Index0Param2Write=true
    Module.Index0Param3Write=true
    Module.Index2Param0Write=true
    Module.Index2Param1Write=true
    Module.Index2Param2Write=true
    Module.Index2Param3Write=true
    Write Parameter Index 2=true
   )
  ]
  rendererClasses [&NiagaraSpriteRendererProperties]
  renderers [
   {rendererClass=&NiagaraSpriteRendererProperties}
   {rendererClass=&NiagaraSpriteRendererProperties, rendererIndex=1}
  ]
 }}
summary {
 systemName=NS_Emote_Stalemate_PropDust
 emitters [{emitterName=DirectionalBurst, rendererClasses [&NiagaraSpriteRendererProperties]}]
 userVariables [User.AlphaScale {type=NiagaraFloat, defaultValue=1, description=""}]
}
