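# Niagara fire effect for the Stalemate emote's hubcap gadget prop, with all three emitters spawning around its JNT_Main_Wheel bone.
# FireSmall draws upward-facing SubUV flame sprites, while GPU sparks scatter through noise forces and stretch with speed; both spawn areas widen over the emitter loop.
# Voxel_On_Med adds animated mesh particles, using a per-particle Burned value to vary material parameters.
# User.AlphaScale scales spark and voxel opacity, but not the flames.
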NiagaraSystem /BRCosmetics/Animation/Game/MainPlayer/Emotes/Stalemate/FX/NS_Emote_Stalemate_PropFire
dependencies {
 usedDataInterfaces [
  &NiagaraDataInterfaceCurve
  &NiagaraDataInterfaceSkeletalMesh
  &NiagaraDataInterfaceColorCurve
  &NiagaraDataInterfaceVector2DCurve
 ]
 usedDynamicInputs [&SimulationPosition, &MakeFloatFromVector]
 usedModules [&SystemState]
}
emitters {
 FireSmall {
  emitterUpdateScript [
   &EmitterState()
   &SpawnRate(Spawn Probability=0.2, SpawnRate=80, Use Spawn Probability=true)
  ]
  particleSpawnScript [
   &InitializeParticle(
    Lifetime=&Multiply_Float(A=&UniformRangedFloat(Minimum=1.2, Maximum=1.5), B=0.5)
    Sprite Rotation Mode=::NewEnumerator2
    Sprite Size Mode=::NewEnumerator3
    Uniform Sprite Size=&MakeFloatFromVector2d(Vector 2D=&Vector2DFromFloat(Value=&Multiply_Float(A=0.2, B=&UniformRangedFloat(Minimum=110, Maximum=145))))
    UsePositionOffset=false
    Write Material Random=true
   ) {module=InitializeParticle001}
   &Color(Color=<rgba 10 0.3707 0 0.55>)
   &DynamicMaterialParameters(
    Index 0 Param 2=&RandomRangeFloat(Minimum=0, Maximum=4)
    Index 0 Param 4=2.5
    Module.Param0WriteEnabled=false
    Module.Param1WriteEnabled=true
    Module.Param2WriteEnabled=false
    Module.Param3WriteEnabled=true
   ) {module=DynamicMaterialParameters001}
   &SkeletalMeshLocation(
    Bone Sampling Mode=::NewEnumerator4
    MaskSpawnGroup=false
    Skeletal Mesh=DataInterface() {propertyValues {
      DefaultMesh=null
      ExcludeBoneName=root
      OptionalComponentClass=null
      PreviewMesh=/BRCosmetics/Animation/Props/Gadgets_Stalemate_Hubcap/Meshes/Stalemate_Hubcap_Gadget
      SkinningMode=SkinOnTheFly
      SoftSourceActor=null
      SourceMode=Default
      UvSetIndex=0
      WholeMeshLOD=-1
      bExcludeBone=false
      bReadDeformedGeometry=true
      bRequireCurrentFrameData=true
      ComponentTags []
      FilteredBones [JNT_Main_Wheel]
      FilteredSockets []
      MeshUserParameter {parameter {name=None, typeDefHandle {registeredTypeIndex=1}, varData []}}
      SamplingRegions []
     }}
   ) {module=SkeletalMeshLocation003}
   &ShapeLocation(
    Box Midpoint=<1 0.5 0.5>
    Box Size=&MakeVector(X=&Lerp_Float(A=15, B=40, Alpha=$Emitter.NormalizedLoopAge), Y=5, Z=0)
    MaskSpawnGroup=false
    Offset=<0 0 5>
    Offset Mode=::NewEnumerator0
    Override Seed=false
    Shape Primitive=::NewEnumerator2
    Surface Only=false
    UseOwnerScale=true
   ) {module=ShapeLocation001}
   &SpriteFacingAndAlignment(Facing Coordinate Space=::World, Sprite Facing=<0 0 1>, WriteAlignment=false, WriteFacing=true)
  ]
  particleUpdateScript [
   &ParticleState()
   /Niagara/Modules/Update/SubUV/SubUVAnimation(Number Of Frames=&FloatFromCurve(
     FloatCurve=DataInterface() {propertyValues {Curve {keys [{arriveTangent=5.9363, interpMode=RCIM_Cubic, leaveTangent=5.9361, tangentMode=RCTM_User, time=0, value=56}, <at 0.9 60>]}}}
     CurveIndex=$Particles.Age
    ))
   &ScaleColor(
    Scale Alpha=&Multiply_Float(
     A=0.5
     B=&FloatFromCurve(FloatCurve=DataInterface() {propertyValues {Curve {keys [
          <at 0 0 RCIM_Cubic>
          {arriveTangent=-0.0116, interpMode=RCIM_Cubic, leaveTangent=-0.0116, tangentMode=RCTM_User, time=0.1, value=1}
          {arriveTangent=-2.0743, interpMode=RCIM_Cubic, leaveTangent=-2.0743, tangentMode=RCTM_User, time=0.6746, value=0.6742}
          {interpMode=RCIM_Cubic, leaveTangent=-2.371, time=1, value=0}
         ]}}})
    )
    Scale RGB=&VectorFromFloat(Value=&FloatFromCurve(FloatCurve=DataInterface() {propertyValues {Curve {keys [<at 0 1.5>,<at 1 0>]}}}, Scale Curve=5))
    ScaleA=true
    ScaleRGB=true
   ) {module=ScaleColor001}
   &ScaleSpriteSize(Scale Factor=&Vector2DFromCurve(
     Vector2Curve=DataInterface() {propertyValues {XCurve {keys [<at 0 0.6>,<at 1 1>]}, YCurve {keys [<at 0 0.6>,<at 1 1>]}}}
     ScaleCurve=&Vector2DFromFloat(Value=&FN_LOD_Distance_Remap(Start Distance=300, End Distance=1000, Minimum Output=1, Maximum Output=1.1))
    ))
   &DynamicMaterialParameters(
    Index 0 Param 1=&FloatFromCurve(FloatCurve=DataInterface() {propertyValues {Curve {keys [<at 0 1>]}}})
    Index 0 Param 2=0.5
    Index 0 Param 3=&FloatFromCurve(FloatCurve=DataInterface() {propertyValues {Curve {keys [
         {arriveTangent=0.3566, interpMode=RCIM_Cubic, leaveTangent=0.3566, tangentMode=RCTM_User, time=0, value=0.25}
         {arriveTangent=1.3428, interpMode=RCIM_Cubic, leaveTangent=1.3428, tangentMode=RCTM_User, time=0.8, value=0.65}
         <at 1 2>
        ]}}})
    Module.Param0WriteEnabled=true
    Module.Param1WriteEnabled=false
    Module.Param2WriteEnabled=true
    Module.Param3WriteEnabled=false
   )
   &SpriteRotationRate(Rotation Rate=&Lerp_Float(A=-10, B=10, Alpha=$Particles.Initial.MaterialRandom)) {module=SpriteRotationRate001}
   &CameraOffset(Camera Offset Amount=-2)
  ]
  rendererClasses [&NiagaraSpriteRendererProperties]
  renderers [{rendererClass=&NiagaraSpriteRendererProperties}]
 }
 NE_FN_Sparks {
  simTarget=GPUComputeSim
  emitterUpdateScript [
   &EmitterState()
   &SpawnRate(SpawnRate=100, Use Spawn Probability=false)
  ]
  particleSpawnScript [
   /SetVariables(Particles.SparkIntensity:NiagaraFloat=&FloatFromCurve(
     FloatCurve=DataInterface() {propertyValues {Curve {keys [
         <at 0 1 RCIM_Cubic>
         {arriveTangent=-2.5317, interpMode=RCIM_Cubic, leaveTangent=-2.5317, tangentMode=RCTM_User, time=0.6, value=0.5}
         <at 1 0 RCIM_Cubic>
        ]}}}
     CurveIndex=&RandomRangeFloat(Minimum=0, Maximum=1)
    )) {isSetParameters=true}
   &InitializeParticle(
    Color=&LerpLinearColors(StartColor=<rgba 10 1 0.1 1>, EndColor=<rgba 25 5 0.2 1>, LerpFactor=$Particles.SparkIntensity)
    Lifetime Max=0.6
    Lifetime Min=0.1
    Lifetime Mode=::NewEnumerator1
    Mass=&Lerp_Float(A=0.3, B=1, Alpha=$Particles.SparkIntensity)
    Mass Mode=::NewEnumerator1
    Sprite Size Mode=::NewEnumerator3
    Uniform Sprite Size=&Lerp_Float(A=0.6, B=1, Alpha=$Particles.SparkIntensity)
    UsePositionOffset=false
    Write Material Random=true
   )
   /SetVariables(Particles.SpriteSize=&Multiply_Vector2(
     A=$Particles.SpriteSize
     B=&Vector2DFromFloat(Value=&FloatFromCurve(FloatCurve=DataInterface() {propertyValues {Curve {keys [<at 1000 1 RCIM_Cubic>,<at 5000 2>]}}}, CurveIndex=$Engine.Owner.LODDistance))
    )) {isSetParameters=true}
   &SkeletalMeshLocation(
    Bone Sampling Mode=::NewEnumerator4
    MaskSpawnGroup=false
    Skeletal Mesh=DataInterface() {propertyValues {
      DefaultMesh=null
      ExcludeBoneName=root
      OptionalComponentClass=null
      PreviewMesh=&Stalemate_Hubcap_Gadget
      SkinningMode=SkinOnTheFly
      SoftSourceActor=null
      SourceMode=Default
      UvSetIndex=0
      WholeMeshLOD=-1
      bExcludeBone=false
      bReadDeformedGeometry=true
      bRequireCurrentFrameData=true
      ComponentTags []
      FilteredBones [JNT_Main_Wheel]
      FilteredSockets []
      MeshUserParameter {parameter {name=None, typeDefHandle {registeredTypeIndex=1}, varData []}}
      SamplingRegions []
     }}
   ) {module=SkeletalMeshLocation003}
   &ShapeLocation(
    Box Midpoint=<1 0.5 0.5>
    Box Size=&MakeVector(X=&Lerp_Float(A=25, B=100, Alpha=$Emitter.NormalizedLoopAge), Y=5, Z=10)
    MaskSpawnGroup=false
    Offset=<0 0 5>
    Offset Mode=::NewEnumerator0
    Override Seed=false
    Shape Primitive=::NewEnumerator2
    Surface Only=false
    UseOwnerScale=true
   ) {module=ShapeLocation001}
   &AddVelocityFromPoint(
    Module.Use Falloff Distance=false
    Origin Offset=<0 0 -50>
    Velocity Strength=&RandomRangeFloat(Minimum=25, Maximum=50)
   )
  ]
  particleUpdateScript [
   &ParticleState()
   &AccelerationForce(Acceleration=&Multiply_Vector(
     A=&VectorFromFloat(Value=1)
     B=&UniformRangedVector(Minimum=<0 0 150>, Maximum=<0 1 250>)
    ))
   &Drag(
    Drag=&FloatFromCurve(FloatCurve=DataInterface() {propertyValues {Curve {keys [<at 0 0.5>,<at 0.5 0.5>,<at 1 1>]}}}, Scale Curve=9)
    Use Linear Drag=true
    Use Rotational Drag=false
   )
   /Niagara/Modules/Update/Forces/CurlNoiseForce(
    Noise Frequency=0.04
    Noise Strength=&FloatFromCurve(
     FloatCurve=DataInterface() {propertyValues {Curve {keys [<at 0 1>,<at 1 0>]}}}
     Scale Curve=&RandomRangeFloat(Minimum=100, Maximum=400)
    )
    Pan Noise=true
   )
   &VectorNoiseForce(
    Force Amount=&FloatFromCurve(
     FloatCurve=DataInterface() {propertyValues {Curve {keys [
         {arriveTangent=-9.4515, interpMode=RCIM_Cubic, leaveTangent=-4.1687, tangentMode=RCTM_User, time=0, value=1}
         {arriveTangent=-3.4372, interpMode=RCIM_Cubic, leaveTangent=-3.4372, tangentMode=RCTM_User, time=0.6, value=0}
        ]}}}
     Scale Curve=1200
    )
    Module.Use Random Noise Time Interval=false
   )
   &SolveForcesAndVelocity(Clamp Velocity=false, Limit Acceleration=false)
   &ScaleColor(
    Scale Alpha=&Multiply_Float(A=$User.AlphaScale, B=1)
    Scale RGB=&MakeVectorFromLinearColorRGB(Color=&ColorFromCurve(ColorCurve=DataInterface() {propertyValues {
        AlphaCurve {keys [<at 0 1>,<at 1 1>]}
        BlueCurve {keys [
          <at 0 2 RCIM_Cubic>
          {arriveTangent=-1.452, interpMode=RCIM_Cubic, leaveTangent=-1.452, tangentMode=RCTM_User, time=0.1467, value=1}
          <at 1 0.0755>
         ]}
        GreenCurve {keys [
          <at 0 2 RCIM_Cubic>
          {arriveTangent=-1.4408, interpMode=RCIM_Cubic, leaveTangent=-1.4408, tangentMode=RCTM_User, time=0.1467, value=1}
          <at 1 0.078>
         ]}
        RedCurve {keys [
          <at 0 2 RCIM_Cubic>
          {arriveTangent=-0.8332, interpMode=RCIM_Cubic, leaveTangent=-0.8332, tangentMode=RCTM_User, time=0.1467, value=1}
          <at 1 0.25>
         ]}
       }}))
    ScaleA=true
    ScaleRGB=true
   )
   &ScaleSpriteSize(Scale Factor=&Vector2DFromFloat(Value=&FloatFromCurve(FloatCurve=DataInterface() {propertyValues {Curve {keys [
          <at 0 0 RCIM_Cubic>
          {arriveTangent=0.1832, interpMode=RCIM_Cubic, leaveTangent=0.1832, tangentMode=RCTM_User, time=0.1, value=1}
          {arriveTangent=-0.2352, interpMode=RCIM_Cubic, leaveTangent=-0.2352, tangentMode=RCTM_User, time=0.4, value=1}
          {arriveTangent=-3.8662, interpMode=RCIM_Cubic, leaveTangent=-3.8662, tangentMode=RCTM_User, time=1, value=0}
         ]}}})))
   &ScaleSpriteSizeBySpeed(
    Max Scale Factor=<1.25 10>
    Sample Scale Factor By Curve=true
    Scale Factor Curve=DataInterface() {propertyValues {Curve {keys [
        {arriveTangent=0.2757, time=0, value=0}
        {arriveTangent=0.3844, interpMode=RCIM_Cubic, leaveTangent=0.3844, tangentMode=RCTM_User, time=0.3, value=0.15}
        {arriveTangent=2.2765, interpMode=RCIM_Cubic, leaveTangent=2.2765, tangentMode=RCTM_User, time=1, value=1}
       ]}}}
    Velocity Threshold=160
   )
  ]
  rendererClasses [&NiagaraSpriteRendererProperties]
  renderers [{rendererClass=&NiagaraSpriteRendererProperties}]
 }
 Voxel_On_Med {
  emitterUpdateScript [
   &EmitterState()
   &SpawnRate(Spawn Probability=0.2, SpawnRate=100, Use Spawn Probability=true)
  ]
  particleSpawnScript [
   &InitializeParticle(
    Color=<rgba 1 1 1 0.5>
    Lifetime=&UniformRangedFloat(Minimum=1, Maximum=1.5)
    Mesh Scale=&Multiply_VectorByFloat(
     Vector=&MakeVector(X=1, Y=1, Z=&RandomRangeFloat(Minimum=0.4, Maximum=0.6))
     Float=&RandomRangeFloat(Minimum=0.35, Maximum=0.5)
    )
    Mesh Scale Mode=::NewEnumerator4
    Sprite Size Mode=::NewEnumerator3
    Uniform Sprite Size=1
    UsePositionOffset=false
    Write Material Random=true
   ) {module=InitializeParticle001}
   /SetVariables(Particles.Burned:NiagaraFloat=&ClampFloat(Float=&FloatFromCurve(FloatCurve=DataInterface() {propertyValues {Curve {keys [
          {arriveTangent=1.1829, interpMode=RCIM_Cubic, leaveTangent=1.1829, tangentMode=RCTM_User, time=0.5, value=0}
          {interpMode=RCIM_Cubic, tangentMode=RCTM_User, time=1, value=1}
         ]}}}))) {isSetParameters=true}
   /SetVariables(Particles.Size:NiagaraFloat=&Multiply_Float(
     A=1
     B=&FloatFromCurve(FloatCurve=DataInterface() {propertyValues {Curve {keys [{interpMode=RCIM_Cubic, tangentMode=RCTM_User, time=0, value=1}, {interpMode=RCIM_Cubic, tangentMode=RCTM_User, time=1, value=0.8}]}}})
    )) {isSetParameters=true}
   /SetVariables(
    Particles.Lifetime=&self:ScratchDynamicInput(Nth:NiagaraInt32=12, ScaleFactor:NiagaraFloat=&RandomRangeFloat(Minimum=1, Maximum=1))
    Particles.Scale=&self:Scale_Nth_IDx(
     Nth:NiagaraInt32=12
     Scale:Vector3f=$Particles.Scale
     ScaleFactor:NiagaraFloat=&RandomRangeFloat(Minimum=1, Maximum=1)
    )
   ) {isSetParameters=true}
   /SetVariables(Particles.Magnitude:NiagaraFloat=&RemapRange(Input Value To Remap=$Particles.Burned, Input Min=0, Input Max=1, Output Min=1, Output Max=0.1)) {isSetParameters=true}
   &SkeletalMeshLocation(
    Bone Sampling Mode=::NewEnumerator4
    MaskSpawnGroup=false
    Orientation Sampling=::NewEnumerator1
    Skeletal Mesh=DataInterface() {propertyValues {
      DefaultMesh=null
      ExcludeBoneName=root
      OptionalComponentClass=null
      PreviewMesh=&Stalemate_Hubcap_Gadget
      SkinningMode=SkinOnTheFly
      SoftSourceActor=null
      SourceMode=Default
      UvSetIndex=0
      WholeMeshLOD=-1
      bExcludeBone=false
      bReadDeformedGeometry=true
      bRequireCurrentFrameData=true
      ComponentTags []
      FilteredBones [JNT_Main_Wheel]
      FilteredSockets []
      MeshUserParameter {parameter {name=None, typeDefHandle {registeredTypeIndex=1}, varData []}}
      SamplingRegions []
     }}
   )
   &ShapeLocation(
    Box Size=<25 5 0>
    MaskSpawnGroup=false
    Offset=&FN_FrontEndVector3Modifier(Gameplay=<0 0 3.8>, Front End=<0 0 3.8>)
    Offset Coordinate Space=::World
    Offset Mode=::NewEnumerator0
    Override Seed=false
    Shape Primitive=::NewEnumerator2
    Surface Only=false
    UseOwnerScale=true
   )
  ]
  particleUpdateScript [
   &ParticleState()
   &ScaleColor(
    Scale Alpha=&Multiply_Float(
     A=&FloatFromCurve(FloatCurve=DataInterface() {propertyValues {Curve {keys [
          {arriveTangent=9.0725, interpMode=RCIM_Cubic, leaveTangent=9.0725, tangentMode=RCTM_User, time=0, value=0}
          {arriveTangent=0.1459, interpMode=RCIM_Cubic, leaveTangent=0.1459, tangentMode=RCTM_User, time=0.1408, value=0.9847}
          {interpMode=RCIM_Cubic, tangentMode=RCTM_User, time=0.4, value=1}
          {arriveTangent=-0.0405, interpMode=RCIM_Cubic, leaveTangent=-0.0405, tangentMode=RCTM_User, time=0.8, value=0}
         ]}}})
     B=$User.AlphaScale
    )
    Scale RGB=&Multiply_VectorByFloat(
     Vector=<1 1 1>
     Float=&FloatFromCurve(FloatCurve=DataInterface() {propertyValues {Curve {keys [
          {arriveTangent=9.0725, interpMode=RCIM_Cubic, leaveTangent=9.0725, tangentMode=RCTM_User, time=0, value=0}
          {arriveTangent=-2.5249, interpMode=RCIM_Cubic, leaveTangent=-2.5249, tangentMode=RCTM_User, time=0.1408, value=0.9847}
          {arriveTangent=-0.0405, interpMode=RCIM_Cubic, leaveTangent=-0.0405, tangentMode=RCTM_User, time=1, value=0.1}
         ]}}})
    )
    ScaleA=true
    ScaleRGB=true
   )
   &ScaleMeshSize(Scale Factor=&VectorFromFloat(Value=&FloatFromCurve(FloatCurve=DataInterface() {propertyValues {Curve {keys [
          <at 0 0.7 RCIM_Cubic>
          {arriveTangent=-0.1095, interpMode=RCIM_Cubic, leaveTangent=-0.1095, tangentMode=RCTM_User, time=0.2, value=0.8}
          {interpMode=RCIM_Cubic, tangentMode=RCTM_User, time=1, value=0.7}
         ]}}}))) {module=ScaleMeshSize001}
   /Niagara/Modules/Update/SubUV/V2/SubUVAnimation(
    Start Frame:NiagaraInt32=&RandomRangeInt(Minimum=0, Maximum=10)
    SubUV Loop Count:NiagaraFloat=&RandomRangeFloat(Minimum=1, Maximum=1.25)
   )
   &DynamicMaterialParameters(
    Index 0 Param 1=&Multiply_Float(
     A=&RemapRange(Input Value To Remap=$Particles.Burned, Input Min=0, Input Max=1, Output Min=1, Output Max=0.2)
     B=&RandomRangeFloat(Minimum=1, Maximum=0.2)
    )
    Index 0 Param 2=&RandomRangeFloat(Minimum=0, Maximum=2)
    Index 0 Param 3=8
    Index 0 Param 4=$Particles.SubImageIndex
    Index 1 Param 1=&RandomRangeFloat(Minimum=-1, Maximum=1)
    Index 1 Param 2=&RandomRangeFloat(Minimum=0, Maximum=3)
    Index 1 Param 3=&FloatFromCurve(
     FloatCurve=DataInterface() {propertyValues {Curve {keys [
         {arriveTangent=4.1785, interpMode=RCIM_Cubic, leaveTangent=4.1785, tangentMode=RCTM_User, time=0, value=0}
         {arriveTangent=0.9995, interpMode=RCIM_Cubic, leaveTangent=0.9995, tangentMode=RCTM_User, time=0.2223, value=0.6825}
         {arriveTangent=-0.0405, interpMode=RCIM_Cubic, leaveTangent=-0.0405, tangentMode=RCTM_User, time=0.7, value=0.9}
         <at 1 0.25>
        ]}}}
     Scale Curve=&RemapRange(Input Value To Remap=$Particles.Burned, Input Min=0, Input Max=1, Output Min=3, Output Max=0.5)
    )
    Module.Param0WriteEnabled=true
    Module.Param0WriteEnabled1=true
    Module.Param1WriteEnabled=true
    Module.Param1WriteEnabled1=true
    Module.Param2WriteEnabled=true
    Module.Param2WriteEnabled1=true
    Module.Param3WriteEnabled=true
    Module.Param3WriteEnabled1=false
    Write Parameter Index 1=true
   )
   &CameraOffset(Camera Offset Amount=-2)
  ]
  rendererClasses [&NiagaraMeshRendererProperties]
  renderers [
   {rendererClass=&NiagaraMeshRendererProperties, rendererIndex=0}
   {rendererClass=&NiagaraMeshRendererProperties, rendererIndex=1}
  ]
 }
}
summary {
 systemName=NS_Emote_Stalemate_PropFire
 emitters [
  {emitterName=Voxel_On_Med, rendererClasses [&NiagaraMeshRendererProperties]}
  {emitterName=NE_FN_Sparks, simTarget=GPUComputeSim, rendererClasses [&NiagaraSpriteRendererProperties]}
  {emitterName=FireSmall, rendererClasses [&NiagaraSpriteRendererProperties]}
 ]
 userVariables [
  User.AlphaScale {type=NiagaraFloat, defaultValue=1, description=""}
  User.bIsFrontEnd {type=NiagaraBool}
  User.bIsFrontEndPreview {type=NiagaraBool}
 ]
}
