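# Niagara system for the Stalemate emote's hubcap gadget prop, adding smoke, flames, sparks and glow.
# Two ribbon emitters trail dark smoke from the JNT_WheelAutoRoll bone, where a burst mesh emitter adds an orange glow.
# Socket-anchored flame sprites, GPU sparks and fire/glow decal sprites cover the prop, with vortex forces centered on FX_Socket_6 swirling flames and sparks.
# User.AlphaScale scales every emitter's opacity.
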NiagaraSystem /BRCosmetics/Animation/Game/MainPlayer/Emotes/Stalemate/FX/NS_Emote_Stalemate_PropSmoke
dependencies {
 usedDataInterfaces [&NiagaraDataInterfaceSkeletalMesh, &NiagaraDataInterfaceCurve, &NiagaraDataInterfaceColorCurve]
 usedDynamicInputs []
 usedModules [&SystemState]
}
emitters {
 DirectionalBurst {
  emitterUpdateScript [
   &EmitterState()
   &SpawnRate(SpawnRate=20, Use Spawn Probability=false)
  ]
  particleSpawnScript [
   &InitializeParticle(
    Color=<rgba 0.0104 0.0104 0.0104 1>
    Ribbon Width=35
    Ribbon Width Mode=::NewEnumerator1
    Sprite Rotation Mode=::NewEnumerator2
    Sprite Size Mode=::NewEnumerator3
    Uniform Sprite Size=25
    UsePositionOffset=false
    Write Material Random=true
   )
   &SkeletalMeshLocation(
    Bone Sampling Mode=::NewEnumerator4
    MaskSpawnGroup=false
    Skeletal Mesh=DataInterface() {propertyValues {
      DefaultMesh=null
      ExcludeBoneName=root
      OptionalComponentClass=null
      PreviewMesh=/BRCosmetics/Animation/Props/Gadgets_Stalemate_Hubcap/Meshes/Stalemate_Hubcap_Gadget
      SkinningMode=SkinOnTheFly
      SoftSourceActor=null
      SourceMode=Default
      UvSetIndex=0
      WholeMeshLOD=-1
      bExcludeBone=false
      bReadDeformedGeometry=true
      bRequireCurrentFrameData=true
      ComponentTags []
      FilteredBones [JNT_WheelAutoRoll]
      FilteredSockets []
      MeshUserParameter {parameter {name=None, typeDefHandle {registeredTypeIndex=1}, varData []}}
      SamplingRegions []
     }}
   )
  ]
  particleUpdateScript [
   &ParticleState()
   &ScaleColor(
    Scale Alpha=&FloatFromCurve(
     FloatCurve=DataInterface() {propertyValues {Curve {keys [
         {arriveTangent=8.2957, interpMode=RCIM_Cubic, leaveTangent=8.2957, tangentMode=RCTM_User, time=0, value=0}
         {arriveTangent=-0.0204, interpMode=RCIM_Cubic, leaveTangent=-0.0204, tangentMode=RCTM_User, time=0.2, value=1.5}
         {arriveTangent=-1.5192, interpMode=RCIM_Cubic, leaveTangent=-1.5192, tangentMode=RCTM_User, time=0.6, value=0.45}
         {arriveTangent=-0.0405, interpMode=RCIM_Cubic, leaveTangent=-0.0405, tangentMode=RCTM_User, time=1, value=0}
        ]}}}
     Scale Curve=&Multiply_Float(A=1, B=$User.AlphaScale)
    )
    Scale RGB=&VectorFromFloat(Value=1)
    ScaleA=true
    ScaleRGB=true
   )
   &ScaleRibbonWidth(Ribbon Width Scale=&FloatFromCurve(
     FloatCurve=DataInterface() {propertyValues {Curve {keys [{interpMode=RCIM_Cubic, tangentMode=RCTM_User, time=0, value=0.5}, {interpMode=RCIM_Cubic, tangentMode=RCTM_User, time=0.2, value=1}]}}}
     Scale Curve=0.8
    ))
   &SkeletalMeshLocation(
    Bone Sampling Mode=::NewEnumerator4
    MaskSpawnGroup=false
    Sampled Position Offset=&MakeVector(X=&Lerp_Float(A=0, B=-80, Alpha=$Particles.NormalizedAge), Y=0, Z=0)
    Sampled Position Offset Coordinate Space:ENiagaraSkelMeshTransforms=::NewEnumerator2
    Skeletal Mesh=DataInterface() {propertyValues {
      DefaultMesh=null
      ExcludeBoneName=root
      OptionalComponentClass=null
      PreviewMesh=&Stalemate_Hubcap_Gadget
      SkinningMode=SkinOnTheFly
      SoftSourceActor=null
      SourceMode=Default
      UvSetIndex=0
      WholeMeshLOD=-1
      bExcludeBone=false
      bReadDeformedGeometry=true
      bRequireCurrentFrameData=true
      ComponentTags []
      FilteredBones [JNT_WheelAutoRoll]
      FilteredSockets []
      MeshUserParameter {parameter {name=None, typeDefHandle {registeredTypeIndex=1}, varData []}}
      SamplingRegions []
     }}
   ) {module=SkeletalMeshLocation001}
  ]
  rendererClasses [&NiagaraRibbonRendererProperties]
  renderers [{rendererClass=&NiagaraRibbonRendererProperties}]
 }
 DirectionalBurst004 {
  emitterUpdateScript [
   &EmitterState()
   &SpawnRate(SpawnRate=20, Use Spawn Probability=false)
  ]
  particleSpawnScript [
   &InitializeParticle(
    Color=<rgba 0.0104 0.0104 0.0104 1>
    Ribbon Width=35
    Ribbon Width Mode=::NewEnumerator1
    Sprite Rotation Mode=::NewEnumerator2
    Sprite Size Mode=::NewEnumerator3
    Uniform Sprite Size=&Multiply_Float(A=6, B=&RandomRangeFloat(Minimum=9, Maximum=13))
    UsePositionOffset=false
    Write Material Random=true
   )
   &SkeletalMeshLocation(
    Bone Sampling Mode=::NewEnumerator4
    MaskSpawnGroup=false
    Skeletal Mesh=DataInterface() {propertyValues {
      DefaultMesh=null
      ExcludeBoneName=root
      OptionalComponentClass=null
      PreviewMesh=&Stalemate_Hubcap_Gadget
      SkinningMode=SkinOnTheFly
      SoftSourceActor=null
      SourceMode=Default
      UvSetIndex=0
      WholeMeshLOD=-1
      bExcludeBone=false
      bReadDeformedGeometry=true
      bRequireCurrentFrameData=true
      ComponentTags []
      FilteredBones [JNT_WheelAutoRoll]
      FilteredSockets []
      MeshUserParameter {parameter {name=None, typeDefHandle {registeredTypeIndex=1}, varData []}}
      SamplingRegions []
     }}
   )
  ]
  particleUpdateScript [
   &ParticleState()
   &ScaleSpriteSize(
    Uniform Curve Scale=0.8
    Uniform Curve Sprite Scale=DataInterface() {propertyValues {Curve {keys [{arriveTangent=1.1766, interpMode=RCIM_Cubic, leaveTangent=1.1766, tangentMode=RCTM_User, time=0, value=0.25}, <at 1 1 RCIM_Cubic>]}}}
   )
   &ScaleColor(
    Scale Alpha=&FloatFromCurve(
     FloatCurve=DataInterface() {propertyValues {Curve {keys [
         {arriveTangent=8.2957, interpMode=RCIM_Cubic, leaveTangent=8.2957, tangentMode=RCTM_User, time=0, value=0}
         {arriveTangent=-0.0204, interpMode=RCIM_Cubic, leaveTangent=-0.0204, tangentMode=RCTM_User, time=0.1, value=1}
         {arriveTangent=-1.5192, interpMode=RCIM_Cubic, leaveTangent=-1.5192, tangentMode=RCTM_User, time=0.6, value=0.45}
         {arriveTangent=-0.0405, interpMode=RCIM_Cubic, leaveTangent=-0.0405, tangentMode=RCTM_User, time=1, value=0}
        ]}}}
     Scale Curve=&Multiply_Float(A=1, B=$User.AlphaScale)
    )
    Scale RGB=&VectorFromFloat(Value=1)
    ScaleA=true
    ScaleRGB=true
   )
   &ScaleRibbonWidth(Ribbon Width Scale=&FloatFromCurve(
     FloatCurve=DataInterface() {propertyValues {Curve {keys [{interpMode=RCIM_Cubic, tangentMode=RCTM_User, time=0, value=0.5}, {interpMode=RCIM_Cubic, tangentMode=RCTM_User, time=0.2, value=1}]}}}
     Scale Curve=1.2
    ))
   &DynamicMaterialParameters(
    Index 0 Param 1=17.6
    Index 0 Param 2=1
    Index 0 Param 3=1
    Module.Index0Param0Write=true
    Module.Index0Param1Write=true
    Module.Index0Param2Write=true
    Module.Index0Param3Write=true
   ) {module=DynamicMaterialParameters001}
   &SkeletalMeshLocation(
    Bone Sampling Mode=::NewEnumerator4
    MaskSpawnGroup=false
    Sampled Position Offset=&MakeVector(X=&Lerp_Float(A=0, B=-80, Alpha=$Particles.NormalizedAge), Y=0, Z=0)
    Sampled Position Offset Coordinate Space:ENiagaraSkelMeshTransforms=::NewEnumerator2
    Skeletal Mesh=DataInterface() {propertyValues {
      DefaultMesh=null
      ExcludeBoneName=root
      OptionalComponentClass=null
      PreviewMesh=&Stalemate_Hubcap_Gadget
      SkinningMode=SkinOnTheFly
      SoftSourceActor=null
      SourceMode=Default
      UvSetIndex=0
      WholeMeshLOD=-1
      bExcludeBone=false
      bReadDeformedGeometry=true
      bRequireCurrentFrameData=true
      ComponentTags []
      FilteredBones [JNT_WheelAutoRoll]
      FilteredSockets []
      MeshUserParameter {parameter {name=None, typeDefHandle {registeredTypeIndex=1}, varData []}}
      SamplingRegions []
     }}
   ) {module=SkeletalMeshLocation001}
  ]
  rendererClasses [&NiagaraRibbonRendererProperties]
  renderers [{rendererClass=&NiagaraRibbonRendererProperties}]
 }
 DirectionalBurst005 {
  emitterUpdateScript [
   &EmitterState(Life Cycle Mode=::NewEnumerator1, Loop Behavior=::NewEnumerator1, Loop Duration Mode=::NewEnumerator1, UseLoopDelay=false)
   &SpawnBurst_Instantaneous(Use Spawn Probability=false, UseLoopCountLimit=false)
  ]
  particleSpawnScript [
   &InitializeParticle(
    Color=<rgba 3 0.5 0 1>
    Lifetime=1.5
    Mesh Scale Mode=::NewEnumerator3
    Sprite Rotation Mode=::NewEnumerator2
    Sprite Size Mode=::NewEnumerator3
    Uniform Sprite Size=20
    UsePositionOffset=false
    Write Material Random=true
   )
   &CameraOffset(Camera Offset Amount=2)
  ]
  particleUpdateScript [
   &ParticleState(Kill Particles When Lifetime Has Elapsed=false, Let Infinitely Lived Particles Die When Emitter Deactivates=true)
   &SkeletalMeshLocation(
    Bone Sampling Mode=::NewEnumerator4
    MaskSpawnGroup=false
    Skeletal Mesh=DataInterface() {propertyValues {
      DefaultMesh=null
      ExcludeBoneName=root
      OptionalComponentClass=null
      PreviewMesh=&Stalemate_Hubcap_Gadget
      SkinningMode=SkinOnTheFly
      SoftSourceActor=null
      SourceMode=Default
      UvSetIndex=0
      WholeMeshLOD=-1
      bExcludeBone=false
      bReadDeformedGeometry=true
      bRequireCurrentFrameData=true
      ComponentTags []
      FilteredBones [JNT_WheelAutoRoll]
      FilteredSockets []
      MeshUserParameter {parameter {name=None, typeDefHandle {registeredTypeIndex=1}, varData []}}
      SamplingRegions []
     }}
   )
   &ScaleColor(Scale Alpha=$User.AlphaScale, Scale RGB=&VectorFromFloat(Value=6), ScaleA=true, ScaleRGB=true)
   &ScaleRibbonWidth(Ribbon Width Scale=&FloatFromCurve(FloatCurve=DataInterface() {propertyValues {Curve {keys [{interpMode=RCIM_Cubic, tangentMode=RCTM_User, time=0, value=0.5}, {interpMode=RCIM_Cubic, tangentMode=RCTM_User, time=0.2, value=1}]}}}))
   &DynamicMaterialParameters(
    Index 0 Param 1=&RandomRangeFloat(Minimum=0, Maximum=1)
    Index 0 Param 2=1
    Index 0 Param 3=10
    Index 1 Param 4=&Multiply_Float(
     A=&FloatFromCurve(
      FloatCurve=DataInterface() {propertyValues {Curve {keys [
          <at 0 1 RCIM_Cubic>
          {arriveTangent=-1.3184, interpMode=RCIM_Cubic, leaveTangent=-1.3184, time=0.0958, value=1.2356}
          {arriveTangent=0.0731, interpMode=RCIM_Cubic, leaveTangent=0.0731, time=0.2145, value=0.7171}
          {arriveTangent=1.6609, interpMode=RCIM_Cubic, leaveTangent=1.6609, time=0.2892, value=1.2497}
          {arriveTangent=0.2524, interpMode=RCIM_Cubic, leaveTangent=0.2524, time=0.3706, value=0.9763}
          {arriveTangent=-0.265, interpMode=RCIM_Cubic, leaveTangent=-0.265, time=0.7188, value=1.3581}
          {arriveTangent=-1.2732, interpMode=RCIM_Cubic, leaveTangent=-1.2732, time=0.8508, value=0.8491}
          <at 1 1 RCIM_Cubic>
         ]}}}
      Scale Curve=0.25
     )
     B=1
    )
    Module.Param0WriteEnabled=true
    Module.Param0WriteEnabled1=true
    Module.Param1WriteEnabled=true
    Module.Param1WriteEnabled1=true
    Module.Param2WriteEnabled=true
    Module.Param2WriteEnabled1=true
    Module.Param3WriteEnabled=true
    Module.Param3WriteEnabled1=true
    Write Parameter Index 1=true
   ) {module=DynamicMaterialParameters001}
  ]
  rendererClasses [&NiagaraMeshRendererProperties]
  renderers [
   {rendererClass=&NiagaraMeshRendererProperties, rendererIndex=0}
   {rendererClass=&NiagaraMeshRendererProperties, rendererIndex=1}
   {rendererClass=&NiagaraMeshRendererProperties, rendererIndex=2}
  ]
 }
 Flames {
  emitterUpdateScript [
   &EmitterState()
   &SpawnRate(SpawnRate=120, Use Spawn Probability=false) {module=SpawnRate002}
  ]
  particleSpawnScript [
   &InitializeParticle(
    Lifetime=&RandomRangeFloat(Minimum=0.1, Maximum=0.3)
    Sprite Size=<2 4>
    Sprite Size Mode=::NewEnumerator4
    Sprite UV Mode=::NewEnumerator1
    UsePositionOffset=false
    Write Material Random=true
   )
   &PartitionParticles(Number Of Partitions=5)
   &CameraOffset(Camera Offset Amount=10)
   &SkeletalMeshLocation(
    Bone / Socket Index=$Particles.PartitionParticles.Partition
    MaskSpawnGroup=false
    Mesh Sampling Type=::NewEnumerator5
    Skeletal Mesh=DataInterface() {propertyValues {
      DefaultMesh=null
      ExcludeBoneName=root
      OptionalComponentClass=null
      PreviewMesh=&Stalemate_Hubcap_Gadget
      SkinningMode=SkinOnTheFly
      SoftSourceActor=null
      SourceMode=Default
      UvSetIndex=0
      WholeMeshLOD=-1
      bExcludeBone=false
      bReadDeformedGeometry=true
      bRequireCurrentFrameData=true
      ComponentTags []
      FilteredBones []
      FilteredSockets [FX_Socket_1,FX_Socket_2,FX_Socket_3,FX_Socket_4,FX_Socket_5]
      MeshUserParameter {parameter {name=None, typeDefHandle {registeredTypeIndex=1}, varData []}}
      SamplingRegions []
     }}
    Socket Sampling Mode=::NewEnumerator5
   ) {module=SkeletalMeshLocation001}
   &ShapeLocation(MaskSpawnGroup=false, Override Seed=false, Sphere Radius=2, UseOwnerScale=true)
   &SkeletalMeshLocation(
    MaskSpawnGroup=false
    Mesh Sampling Type=::NewEnumerator5
    Position Sampling=::NewEnumerator2
    Skeletal Mesh=DataInterface() {propertyValues {
      DefaultMesh=null
      ExcludeBoneName=root
      OptionalComponentClass=null
      PreviewMesh=&Stalemate_Hubcap_Gadget
      SkinningMode=SkinOnTheFly
      SoftSourceActor=null
      SourceMode=Default
      UvSetIndex=0
      WholeMeshLOD=-1
      bExcludeBone=false
      bReadDeformedGeometry=true
      bRequireCurrentFrameData=true
      ComponentTags []
      FilteredBones []
      FilteredSockets [FX_Socket_6]
      MeshUserParameter {parameter {name=None, typeDefHandle {registeredTypeIndex=1}, varData []}}
      SamplingRegions []
     }}
    Socket Sampling Mode=::NewEnumerator5
   ) {module=SkeletalMeshLocation002}
   &AddVelocity(
    Module.Use Falloff Distance=false
    Override Seed=false
    Velocity Mode=::NewEnumerator1
    Velocity Origin=$Particles.SkeletalMeshLocation002.SampledPosition
    Velocity Speed=&RandomRangeFloat(Minimum=48, Maximum=60)
   )
   &VortexForce(
    Origin Pull Amount=80
    UseOriginOffset=false
    Vortex Axis=<0 1 0>
    Vortex Force Amount=-100
    Vortex Origin:NiagaraPosition=$Particles.SkeletalMeshLocation002.SampledPosition
   ) {module=VortexForce001}
   &ApplyInitialForces()
  ]
  particleUpdateScript [
   &ParticleState()
   &ScaleSpriteSize(
    Scale Factor=&Multiply_Vector2D(A=&RandomRangeVector2D(Minimum=<2 2>, Maximum=<3 3>), B=<1 1>)
    Scale Sprite Size Mode=::NewEnumerator2
   )
   &SkeletalMeshLocation(
    Initialize Particle Position To Sampled Position=false
    MaskSpawnGroup=false
    Mesh Sampling Type=::NewEnumerator5
    Position Sampling=::NewEnumerator1
    Skeletal Mesh=DataInterface() {propertyValues {
      DefaultMesh=null
      ExcludeBoneName=root
      OptionalComponentClass=null
      PreviewMesh=&Stalemate_Hubcap_Gadget
      SkinningMode=SkinOnTheFly
      SoftSourceActor=null
      SourceMode=Default
      UvSetIndex=0
      WholeMeshLOD=-1
      bExcludeBone=false
      bReadDeformedGeometry=true
      bRequireCurrentFrameData=true
      ComponentTags []
      FilteredBones []
      FilteredSockets [FX_Socket_6]
      MeshUserParameter {parameter {name=None, typeDefHandle {registeredTypeIndex=1}, varData []}}
      SamplingRegions []
     }}
    Socket Sampling Mode=::NewEnumerator5
    Source Movement Scale:Vector3f=<0.5 0.5 0.5>
   ) {module=SkeletalMeshLocation004}
   &VortexForce(
    Origin Pull Amount=20
    UseOriginOffset=false
    Vortex Axis=<0 1 0>
    Vortex Force Amount=-100
    Vortex Origin:NiagaraPosition=$Particles.SkeletalMeshLocation004.SampledPosition
   )
   &Drag(Drag=3, Use Linear Drag=true, Use Rotational Drag=false)
   &SolveForcesAndVelocity(Clamp Velocity=false, Limit Acceleration=false)
   &ScaleSpriteSizeBySpeed(Max Scale Factor=<1 4>)
   &ScaleColor(
    Scale Alpha=&Multiply_Float(
     A=&FloatFromCurve(
      FloatCurve=DataInterface() {propertyValues {Curve {keys [
          {arriveTangent=9.0725, interpMode=RCIM_Cubic, leaveTangent=9.0725, tangentMode=RCTM_User, time=0, value=1}
          {arriveTangent=-2.5249, interpMode=RCIM_Cubic, leaveTangent=-2.5249, tangentMode=RCTM_User, time=0.1408, value=0.9847}
          {arriveTangent=-1.8607, interpMode=RCIM_Cubic, leaveTangent=-1.8607, tangentMode=RCTM_User, time=0.3012, value=0.9413}
          {arriveTangent=-1.0655, interpMode=RCIM_Cubic, leaveTangent=-1.0655, tangentMode=RCTM_User, time=0.5976, value=0.5305}
          {arriveTangent=-0.0405, interpMode=RCIM_Cubic, leaveTangent=-0.0405, tangentMode=RCTM_User, time=1, value=0}
         ]}}}
      Scale Curve=1.5
     )
     B=&Multiply_Float(A=$User.AlphaScale, B=2)
    )
    Scale RGB=&Multiply_VectorByFloat(
     Vector=&MakeVectorFromLinearColorRGB(Color=&ColorFromCurve(ColorCurve=DataInterface() {propertyValues {
         AlphaCurve {keys [<at 0 1 RCIM_Cubic>,<at 1 1 RCIM_Cubic>]}
         BlueCurve {keys [<at 0.001 0>, <at 0.1259 0>, {interpMode=RCIM_Cubic, leaveTangent=0.096, time=0.438, value=0}]}
         GreenCurve {keys [<at 0.001 0.2016>, <at 0.1259 0.2016>, {interpMode=RCIM_Cubic, leaveTangent=-0.2733, time=0.438, value=0.0444}]}
         RedCurve {keys [<at 0.001 1>, <at 0.1259 1>, {interpMode=RCIM_Cubic, leaveTangent=-1.7277, time=0.438, value=0.444}]}
        }}))
     Float=&Multiply_Float(
      A=&Lerp_Float(
       A=0.5
       B=1
       Alpha=&FloatFromCurve(
        FloatCurve=DataInterface() {propertyValues {Curve {keys [
            {arriveTangent=0.0028, interpMode=RCIM_Cubic, leaveTangent=0.0028, tangentMode=RCTM_User, time=0, value=0}
            {arriveTangent=-2.5249, interpMode=RCIM_Cubic, leaveTangent=-2.5249, tangentMode=RCTM_User, time=0.1408, value=0.9847}
            {arriveTangent=-0.0405, interpMode=RCIM_Cubic, leaveTangent=-0.0405, tangentMode=RCTM_User, time=0.8, value=0}
           ]}}}
        Scale Curve=3
       )
      )
      B=&FNTimeOfDayScalar(Night Scale=0.945, Morning Scale=0.7, Evening Scale=0.7)
     )
    )
    ScaleA=true
    ScaleRGB=true
   ) {module=ScaleColor001}
   &DynamicMaterialParameters(
    Index 0 Param 1=&FloatFromCurve(FloatCurve=DataInterface() {propertyValues {Curve {keys [
         {arriveTangent=-0.1846, time=0, value=0}
         {arriveTangent=1.3385, interpMode=RCIM_Cubic, leaveTangent=1.1594, tangentMode=RCTM_Break, time=0.5, value=0}
         {arriveTangent=2, interpMode=RCIM_Cubic, leaveTangent=2, time=0.9, value=0.45}
         {arriveTangent=3.0091, interpMode=RCIM_Cubic, leaveTangent=3.0091, tangentMode=RCTM_User, time=1, value=1}
        ]}}})
    Index 0 Param 2=1
    Index 0 Param 3=&Lerp_Float(A=0, B=1, Alpha=&Multiply_Float(A=$System.EffectIntensity, B=1))
    Index 0 Param 4=&Lerp_Float(A=0, B=1, Alpha=&Multiply_Float(A=$System.EffectIntensity, B=1))
    Index 1 Param 1=&Lerp_Float(A=1, B=0, Alpha=$System.EffectIntensity)
    Module.Index0Param0Write=true
    Module.Index0Param1Write=true
    Module.Index0Param2Write=true
    Module.Index0Param3Write=true
    Module.Index1Param0Write=false
    Module.Index1Param1Write=false
    Module.Index1Param2Write=false
    Module.Index1Param3Write=false
    Write Parameter Index 1=true
   )
   &SkeletalMeshLocation(
    Bone / Socket Index=$Particles.PartitionParticles.Partition
    Initialize Particle Position To Sampled Position=false
    MaskSpawnGroup=false
    Mesh Sampling Type=::NewEnumerator5
    Position Sampling=::NewEnumerator1
    Skeletal Mesh=DataInterface() {propertyValues {
      DefaultMesh=null
      ExcludeBoneName=root
      OptionalComponentClass=null
      PreviewMesh=&Stalemate_Hubcap_Gadget
      SkinningMode=SkinOnTheFly
      SoftSourceActor=null
      SourceMode=Default
      UvSetIndex=0
      WholeMeshLOD=-1
      bExcludeBone=false
      bReadDeformedGeometry=true
      bRequireCurrentFrameData=true
      ComponentTags []
      FilteredBones []
      FilteredSockets [FX_Socket_1,FX_Socket_2,FX_Socket_3,FX_Socket_4,FX_Socket_5]
      MeshUserParameter {parameter {name=None, typeDefHandle {registeredTypeIndex=1}, varData []}}
      SamplingRegions []
     }}
    Socket Sampling Mode=::NewEnumerator5
   ) {module=SkeletalMeshLocation003}
  ]
  rendererClasses [&NiagaraSpriteRendererProperties]
  renderers [{rendererClass=&NiagaraSpriteRendererProperties}]
 }
 NE_FN_Sparks {
  simTarget=GPUComputeSim
  emitterUpdateScript [
   &EmitterState()
   &SpawnRate(SpawnRate=200, Use Spawn Probability=false)
  ]
  particleSpawnScript [
   /SetVariables(Particles.SparkIntensity:NiagaraFloat=&FloatFromCurve(
     FloatCurve=DataInterface() {propertyValues {Curve {keys [
         <at 0 1 RCIM_Cubic>
         {arriveTangent=-2.5317, interpMode=RCIM_Cubic, leaveTangent=-2.5317, tangentMode=RCTM_User, time=0.6, value=0.5}
         <at 1 0 RCIM_Cubic>
        ]}}}
     CurveIndex=&RandomRangeFloat(Minimum=0, Maximum=1)
    )) {isSetParameters=true}
   &InitializeParticle(
    Color=&LerpLinearColors(StartColor=<rgba 10 1 0.1 1>, EndColor=<rgba 25 5 0.2 1>, LerpFactor=$Particles.SparkIntensity)
    Lifetime Max=0.4
    Lifetime Min=0.2
    Lifetime Mode=::NewEnumerator1
    Mass=&Lerp_Float(A=0.3, B=1, Alpha=$Particles.SparkIntensity)
    Mass Mode=::NewEnumerator1
    Sprite Size Mode=::NewEnumerator3
    Uniform Sprite Size=&Lerp_Float(A=0.6, B=1, Alpha=$Particles.SparkIntensity)
    UsePositionOffset=false
    Write Material Random=true
   )
   /SetVariables(Particles.SpriteSize=&Multiply_Vector2(
     A=$Particles.SpriteSize
     B=&Vector2DFromFloat(Value=&FloatFromCurve(FloatCurve=DataInterface() {propertyValues {Curve {keys [<at 1000 1 RCIM_Cubic>,<at 5000 2>]}}}, CurveIndex=$Engine.Owner.LODDistance))
    )) {isSetParameters=true}
   &SkeletalMeshLocation(
    MaskSpawnGroup=false
    Mesh Sampling Type=::NewEnumerator5
    OverrideRandomSeed=false
    Skeletal Mesh=DataInterface() {propertyValues {
      DefaultMesh=null
      ExcludeBoneName=root
      OptionalComponentClass=null
      PreviewMesh=&Stalemate_Hubcap_Gadget
      SkinningMode=SkinOnTheFly
      SoftSourceActor=null
      SourceMode=Default
      UvSetIndex=0
      WholeMeshLOD=-1
      bExcludeBone=false
      bReadDeformedGeometry=true
      bRequireCurrentFrameData=true
      ComponentTags []
      FilteredBones []
      FilteredSockets [FX_Socket_1,FX_Socket_2,FX_Socket_3,FX_Socket_4,FX_Socket_5]
      MeshUserParameter {parameter {name=None, typeDefHandle {registeredTypeIndex=1}, varData []}}
      SamplingRegions []
     }}
   )
   &ShapeLocation(MaskSpawnGroup=false, Override Seed=false, Sphere Radius=5, UseOwnerScale=true)
   &SkeletalMeshLocation(
    MaskSpawnGroup=false
    Mesh Sampling Type=::NewEnumerator5
    Position Sampling=::NewEnumerator2
    Skeletal Mesh=DataInterface() {propertyValues {
      DefaultMesh=null
      ExcludeBoneName=root
      OptionalComponentClass=null
      PreviewMesh=&Stalemate_Hubcap_Gadget
      SkinningMode=SkinOnTheFly
      SoftSourceActor=null
      SourceMode=Default
      UvSetIndex=0
      WholeMeshLOD=-1
      bExcludeBone=false
      bReadDeformedGeometry=true
      bRequireCurrentFrameData=true
      ComponentTags []
      FilteredBones []
      FilteredSockets [FX_Socket_6]
      MeshUserParameter {parameter {name=None, typeDefHandle {registeredTypeIndex=1}, varData []}}
      SamplingRegions []
     }}
    Socket Sampling Mode=::NewEnumerator5
   ) {module=SkeletalMeshLocation001}
  ]
  particleUpdateScript [
   &ParticleState()
   &SkeletalMeshLocation(
    Initialize Particle Position To Sampled Position=false
    MaskSpawnGroup=false
    Mesh Sampling Type=::NewEnumerator5
    Position Sampling=::NewEnumerator1
    Skeletal Mesh=DataInterface() {propertyValues {
      DefaultMesh=null
      ExcludeBoneName=root
      OptionalComponentClass=null
      PreviewMesh=&Stalemate_Hubcap_Gadget
      SkinningMode=SkinOnTheFly
      SoftSourceActor=null
      SourceMode=Default
      UvSetIndex=0
      WholeMeshLOD=-1
      bExcludeBone=false
      bReadDeformedGeometry=true
      bRequireCurrentFrameData=true
      ComponentTags []
      FilteredBones []
      FilteredSockets [FX_Socket_6]
      MeshUserParameter {parameter {name=None, typeDefHandle {registeredTypeIndex=1}, varData []}}
      SamplingRegions []
     }}
    Socket Sampling Mode=::NewEnumerator5
    Source Movement Scale:Vector3f=<0.7 0.7 0.7>
   ) {module=SkeletalMeshLocation002}
   &VortexForce(
    Origin Pull Amount=20
    UseOriginOffset=false
    Vortex Axis=<0 1 0>
    Vortex Force Amount=-150
    Vortex Origin:NiagaraPosition=$Particles.SkeletalMeshLocation002.SampledPosition
   )
   &Drag(Drag=0.15, Use Linear Drag=true, Use Rotational Drag=false)
   &VectorNoiseForce(
    Force Amount=&FloatFromCurve(
     FloatCurve=DataInterface() {propertyValues {Curve {keys [
         {arriveTangent=-9.4515, interpMode=RCIM_Cubic, leaveTangent=-4.1687, tangentMode=RCTM_User, time=0, value=1}
         {arriveTangent=-3.4372, interpMode=RCIM_Cubic, leaveTangent=-3.4372, tangentMode=RCTM_User, time=0.6, value=0}
        ]}}}
     Scale Curve=500
    )
    Module.Use Random Noise Time Interval=false
   )
   &SolveForcesAndVelocity(Clamp Velocity=false, Limit Acceleration=false)
   &ScaleColor(
    Scale Alpha=&Multiply_Float(A=$User.AlphaScale, B=1)
    Scale RGB=&MakeVectorFromLinearColorRGB(Color=&ColorFromCurve(ColorCurve=DataInterface() {propertyValues {
        AlphaCurve {keys [<at 0 1>,<at 1 1>]}
        BlueCurve {keys [
          <at 0 2 RCIM_Cubic>
          {arriveTangent=-1.452, interpMode=RCIM_Cubic, leaveTangent=-1.452, tangentMode=RCTM_User, time=0.1467, value=1}
          <at 1 0.0755>
         ]}
        GreenCurve {keys [
          <at 0 2 RCIM_Cubic>
          {arriveTangent=-1.4408, interpMode=RCIM_Cubic, leaveTangent=-1.4408, tangentMode=RCTM_User, time=0.1467, value=1}
          <at 1 0.078>
         ]}
        RedCurve {keys [
          <at 0 2 RCIM_Cubic>
          {arriveTangent=-0.8332, interpMode=RCIM_Cubic, leaveTangent=-0.8332, tangentMode=RCTM_User, time=0.1467, value=1}
          <at 1 0.25>
         ]}
       }}))
    ScaleA=true
    ScaleRGB=true
   )
   &ScaleSpriteSize(Scale Factor=&Vector2DFromFloat(Value=&FloatFromCurve(FloatCurve=DataInterface() {propertyValues {Curve {keys [
          <at 0 0 RCIM_Cubic>
          {arriveTangent=0.1832, interpMode=RCIM_Cubic, leaveTangent=0.1832, tangentMode=RCTM_User, time=0.1, value=1}
          {arriveTangent=-0.2352, interpMode=RCIM_Cubic, leaveTangent=-0.2352, tangentMode=RCTM_User, time=0.4, value=1}
          {arriveTangent=-3.8662, interpMode=RCIM_Cubic, leaveTangent=-3.8662, tangentMode=RCTM_User, time=1, value=0}
         ]}}})))
   &ScaleSpriteSizeBySpeed(
    Max Scale Factor=<1.25 10>
    Sample Scale Factor By Curve=true
    Scale Factor Curve=DataInterface() {propertyValues {Curve {keys [
        {arriveTangent=0.2757, time=0, value=0}
        {arriveTangent=0.3844, interpMode=RCIM_Cubic, leaveTangent=0.3844, tangentMode=RCTM_User, time=0.3, value=0.15}
        {arriveTangent=2.2765, interpMode=RCIM_Cubic, leaveTangent=2.2765, tangentMode=RCTM_User, time=1, value=1}
       ]}}}
    Velocity Threshold=160
   )
  ]
  rendererClasses [&NiagaraSpriteRendererProperties]
  renderers [{rendererClass=&NiagaraSpriteRendererProperties}]
 }
 firedecal {
  emitterUpdateScript [
   &EmitterState(Life Cycle Mode=::NewEnumerator1, Loop Behavior=::NewEnumerator1, Loop Duration Mode=::NewEnumerator1, UseLoopDelay=false)
   &SpawnBurst_Instantaneous(Spawn Count=5, Use Spawn Probability=false, UseLoopCountLimit=false)
  ]
  particleSpawnScript [
   &InitializeParticle(
    Color=<rgba 10 1.381 0 0.5>
    Lifetime=1.5
    Mesh Scale Mode=::NewEnumerator3
    Sprite Rotation Mode=::NewEnumerator2
    Sprite Size Mode=::NewEnumerator3
    Uniform Sprite Size=15
    UsePositionOffset=false
    Write Material Random=true
   )
   &CameraOffset(Camera Offset Amount=10)
  ]
  particleUpdateScript [
   &ParticleState(Kill Particles When Lifetime Has Elapsed=false, Let Infinitely Lived Particles Die When Emitter Deactivates=true)
   &SkeletalMeshLocation(
    Bone / Socket Index=&ReturnExecIndex()
    MaskSpawnGroup=false
    Mesh Sampling Type=::NewEnumerator5
    Skeletal Mesh=DataInterface() {propertyValues {
      DefaultMesh=null
      ExcludeBoneName=root
      OptionalComponentClass=null
      PreviewMesh=&Stalemate_Hubcap_Gadget
      SkinningMode=SkinOnTheFly
      SoftSourceActor=null
      SourceMode=Default
      UvSetIndex=0
      WholeMeshLOD=-1
      bExcludeBone=false
      bReadDeformedGeometry=true
      bRequireCurrentFrameData=true
      ComponentTags []
      FilteredBones []
      FilteredSockets [FX_Socket_1,FX_Socket_2,FX_Socket_3,FX_Socket_4,FX_Socket_5]
      MeshUserParameter {parameter {name=None, typeDefHandle {registeredTypeIndex=1}, varData []}}
      SamplingRegions []
     }}
    Socket Sampling Mode=::NewEnumerator5
   )
   &ScaleColor(
    Scale Alpha=&Multiply_Float(A=$User.AlphaScale, B=1)
    Scale RGB=&VectorFromFloat(Value=5)
    ScaleA=true
    ScaleRGB=true
   )
   &ScaleRibbonWidth(Ribbon Width Scale=&FloatFromCurve(FloatCurve=DataInterface() {propertyValues {Curve {keys [{interpMode=RCIM_Cubic, tangentMode=RCTM_User, time=0, value=0.5}, {interpMode=RCIM_Cubic, tangentMode=RCTM_User, time=0.2, value=1}]}}}))
   &DynamicMaterialParameters(
    Index 0 Param 1=0.5
    Index 0 Param 2=0.8
    Index 0 Param 3=&RandomRangeFloat(Minimum=0, Maximum=1)
    Index 0 Param 4=&RandomRangeFloat(Minimum=0, Maximum=1)
    Module.Param0WriteEnabled=true
    Module.Param1WriteEnabled=true
    Module.Param2WriteEnabled=true
    Module.Param3WriteEnabled=true
   ) {module=DynamicMaterialParameters002}
  ]
  rendererClasses [&NiagaraSpriteRendererProperties]
  renderers [{rendererClass=&NiagaraSpriteRendererProperties}]
 }
 glowdecal {
  emitterUpdateScript [
   &EmitterState(Life Cycle Mode=::NewEnumerator1, Loop Behavior=::NewEnumerator1, Loop Duration Mode=::NewEnumerator1, UseLoopDelay=false)
   &SpawnBurst_Instantaneous(Spawn Count=6, Use Spawn Probability=false, UseLoopCountLimit=false)
  ]
  particleSpawnScript [
   &InitializeParticle(
    Color=<rgba 10 1.381 0 1>
    Lifetime=1.5
    Mesh Scale Mode=::NewEnumerator3
    Sprite Rotation Mode=::NewEnumerator2
    Sprite Size Mode=::NewEnumerator3
    Uniform Sprite Size=20
    UsePositionOffset=false
    Write Material Random=true
   )
   &CameraOffset(Camera Offset Amount=2)
  ]
  particleUpdateScript [
   &ParticleState(Kill Particles When Lifetime Has Elapsed=false, Let Infinitely Lived Particles Die When Emitter Deactivates=true)
   &SkeletalMeshLocation(
    Bone / Socket Index=&ReturnExecIndex()
    MaskSpawnGroup=false
    Mesh Sampling Type=::NewEnumerator5
    Skeletal Mesh=DataInterface() {propertyValues {
      DefaultMesh=null
      ExcludeBoneName=root
      OptionalComponentClass=null
      PreviewMesh=&Stalemate_Hubcap_Gadget
      SkinningMode=SkinOnTheFly
      SoftSourceActor=null
      SourceMode=Default
      UvSetIndex=0
      WholeMeshLOD=-1
      bExcludeBone=false
      bReadDeformedGeometry=true
      bRequireCurrentFrameData=true
      ComponentTags []
      FilteredBones []
      FilteredSockets [FX_Socket_1,FX_Socket_2,FX_Socket_3,FX_Socket_4,FX_Socket_5,FX_Socket_7]
      MeshUserParameter {parameter {name=None, typeDefHandle {registeredTypeIndex=1}, varData []}}
      SamplingRegions []
     }}
    Socket Sampling Mode=::NewEnumerator5
   )
   &ScaleColor(
    Scale Alpha=&Multiply_Float(A=$User.AlphaScale, B=0.3)
    Scale RGB=&VectorFromFloat(Value=1)
    ScaleA=true
    ScaleRGB=true
   )
   &ScaleRibbonWidth(Ribbon Width Scale=&FloatFromCurve(FloatCurve=DataInterface() {propertyValues {Curve {keys [{interpMode=RCIM_Cubic, tangentMode=RCTM_User, time=0, value=0.5}, {interpMode=RCIM_Cubic, tangentMode=RCTM_User, time=0.2, value=1}]}}}))
   &DynamicMaterialParameters(
    Index 0 Param 1=2
    Index 0 Param 2=0.8
    Index 0 Param 3=&RandomRangeFloat(Minimum=0, Maximum=1)
    Index 0 Param 4=&RandomRangeFloat(Minimum=0, Maximum=1)
    Module.Param0WriteEnabled=true
    Module.Param1WriteEnabled=true
    Module.Param2WriteEnabled=true
    Module.Param3WriteEnabled=true
   ) {module=DynamicMaterialParameters002}
  ]
  rendererClasses [&NiagaraSpriteRendererProperties]
  renderers [{rendererClass=&NiagaraSpriteRendererProperties}]
 }
}
summary {
 systemName=NS_Emote_Stalemate_PropSmoke
 emitters [
  {emitterName=DirectionalBurst, rendererClasses [&NiagaraRibbonRendererProperties]}
  {emitterName=firedecal, rendererClasses [&NiagaraSpriteRendererProperties]}
  {emitterName=NE_FN_Sparks, simTarget=GPUComputeSim, rendererClasses [&NiagaraSpriteRendererProperties]}
  {emitterName=glowdecal, rendererClasses [&NiagaraSpriteRendererProperties]}
  {emitterName=DirectionalBurst004, rendererClasses [&NiagaraRibbonRendererProperties]}
  {emitterName=Flames, rendererClasses [&NiagaraSpriteRendererProperties]}
  {emitterName=DirectionalBurst005, rendererClasses [&NiagaraMeshRendererProperties]}
 ]
 userVariables [
  User.AlphaScale {type=NiagaraFloat, defaultValue=1, description=""}
  User.bIsFrontEnd {type=NiagaraBool}
  User.bIsFrontEndPreview {type=NiagaraBool}
 ]
}
