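# Niagara system for the Stalemate emote that plays a constant wind effect blowing along negative Y.
# Two Straightwind mesh emitters draw faint time-of-day-tinted streaks, animated by dynamic material parameters and stepped by a fixed position offset each update rather than velocity.
# Leaves_Vortex001 adds tumbling mesh leaves with curl noise and rising drag.
# DirectionalBurst001 emits soft sprite puffs in short randomized loops from random offsets, driven by drag, fades and material parameters.
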NiagaraSystem /BRCosmetics/Animation/Game/MainPlayer/Emotes/Stalemate/FX/NS_Emote_Stalemate_WindConstant
dependencies {usedDataInterfaces [&NiagaraDataInterfaceCurve], usedDynamicInputs [], usedModules [&SystemState]}
emitters {
 DirectionalBurst001 {
  emitterUpdateScript [
   &EmitterState(
    Life Cycle Mode=::NewEnumerator1
    Loop Delay=&RandomRangeFloat(Minimum=0.5, Maximum=0.8)
    Loop Duration=&RandomRangeFloat(Minimum=0.2, Maximum=0.3)
    Recalculate Duration Each Loop=true
    UseLoopDelay=true
   )
   /SetVariables(Emitter.RandVec:Vector3f=&RandomRangeVector(Minimum=<-100 0 0>, Maximum=<100 50 0>)) {isSetParameters=true}
   &SpawnRate(SpawnRate=20, Use Spawn Probability=false)
  ]
  particleSpawnScript [
   &InitializeParticle(
    Color=<rgba 1 0.7803 0.65 0.5>
    Lifetime Max=3.5
    Lifetime Min=2.5
    Lifetime Mode=::NewEnumerator1
    Mass Max=2
    Mass Mode=::NewEnumerator3
    Sprite Rotation Mode=::NewEnumerator2
    Sprite Size Mode=::NewEnumerator3
    Uniform Sprite Size=&Multiply_Float(A=6, B=&RandomRangeFloat(Minimum=9, Maximum=13))
    UsePositionOffset=false
    Write Material Random=true
   )
   &ShapeLocation(
    Disc Coverage=1
    MaskSpawnGroup=false
    Offset=&AddVectorToPosition(Position=<0 0 10>, Vector=$Emitter.RandVec)
    Offset Mode=::NewEnumerator0
    Override Seed=false
    Ring Radius=20
    Rotation Angle=90
    Rotation Angle Type:ENiagara_AngleInput=::NewEnumerator0
    Rotation Axis=<0 0 0>
    Rotation Coordinate Space=::World
    Rotation Mode=::NewEnumerator1
    Shape Primitive=::NewEnumerator4
    UseOwnerScale=true
   ) {module=ShapeLocation001}
   &AddVelocity(
    Cone Angle=0
    Cone Axis=<0 -1 0>
    Module.UseConeFalloff=false
    Module.UseVelDistribution=false
    Override Seed=false
    Velocity Mode=::NewEnumerator2
    Velocity Speed=&Multiply_Float(A=2, B=&RandomRangeFloat(Minimum=35, Maximum=15))
   ) {module=AddVelocity001}
  ]
  particleUpdateScript [
   &ParticleState()
   &ScaleSpriteSize(Uniform Curve Sprite Scale=DataInterface() {propertyValues {Curve {keys [{arriveTangent=1.1766, interpMode=RCIM_Cubic, leaveTangent=1.1766, tangentMode=RCTM_User, time=0, value=0.25}, <at 1 1 RCIM_Cubic>]}}})
   &Drag(Drag=&RandomRangeFloat(Minimum=1, Maximum=2), Use Linear Drag=true, Use Rotational Drag=false)
   &ScaleColor(
    Scale Alpha=&FloatFromCurve(
     FloatCurve=DataInterface() {propertyValues {Curve {keys [
         {arriveTangent=0.0125, interpMode=RCIM_Cubic, leaveTangent=0.0125, tangentMode=RCTM_User, time=0, value=0}
         {arriveTangent=0.0105, interpMode=RCIM_Cubic, leaveTangent=0.0105, tangentMode=RCTM_User, time=0.5, value=1}
         {arriveTangent=-0.0405, interpMode=RCIM_Cubic, leaveTangent=-0.0405, tangentMode=RCTM_User, time=1, value=0}
        ]}}}
     Scale Curve=0.4
    )
    Scale RGB=&VectorFromFloat(Value=0.25)
    ScaleA=true
    ScaleRGB=true
   )
   &SolveForcesAndVelocity(Clamp Velocity=false, Limit Acceleration=false)
   &DynamicMaterialParameters(
    Index 0 Param 1=&Add_Float(
     A=&RandomRangeFloat(Minimum=0, Maximum=1)
     B=&FloatFromCurve(
      FloatCurve=DataInterface() {propertyValues {Curve {keys [
          <at 0 0 RCIM_Cubic>
          {arriveTangent=1.5947, interpMode=RCIM_Cubic, leaveTangent=1.5947, tangentMode=RCTM_User, time=0.2557, value=0.5257}
          <at 1 1 RCIM_Cubic>
         ]}}}
      CurveIndex=$Particles.Age
      Scale Curve=0.25
     )
    )
    Index 2 Param 1=&FloatFromCurve(FloatCurve=DataInterface() {propertyValues {Curve {keys [{interpMode=RCIM_Cubic, tangentMode=RCTM_User, time=0.25, value=0}, {interpMode=RCIM_Cubic, tangentMode=RCTM_User, time=0.9, value=1}]}}})
    Index 2 Param 2=&RandomRangeFloat(Minimum=0, Maximum=1)
    Index 2 Param 3=&Add_Float(
     A=1
     B=&FloatFromCurve(FloatCurve=DataInterface() {propertyValues {Curve {keys [{tangentMode=RCTM_User, time=0, value=0}, {tangentMode=RCTM_User, time=1, value=1}]}}})
    )
    Index 2 Param 4=1.5
    Module.Index0Param0Write=true
    Module.Index0Param1Write=true
    Module.Index0Param2Write=true
    Module.Index0Param3Write=true
    Module.Index2Param0Write=true
    Module.Index2Param1Write=true
    Module.Index2Param2Write=true
    Module.Index2Param3Write=true
    Write Parameter Index 2=true
   )
  ]
  rendererClasses [&NiagaraSpriteRendererProperties]
  renderers [
   {rendererClass=&NiagaraSpriteRendererProperties}
   {rendererClass=&NiagaraSpriteRendererProperties, rendererIndex=1}
  ]
 }
 Leaves_Vortex001 {
  emitterUpdateScript [
   &EmitterState()
   &SpawnRate(SpawnRate=2, Use Spawn Probability=false)
  ]
  particleSpawnScript [
   &InitializeParticle(
    Color=<rgba 2.5 0.25 0 1>
    Lifetime Max=2.5
    Lifetime Min=1.2
    Lifetime Mode=::NewEnumerator1
    Mesh Scale Mode=::NewEnumerator1
    Mesh Uniform Scale Max=1.1
    Mesh Uniform Scale Min=0.8
    Sprite Size Mode=::NewEnumerator3
    Uniform Sprite Size=&RandomRangeFloat(Minimum=16, Maximum=21)
    UsePositionOffset=false
    Write Material Random=true
   )
   &ShapeLocation(
    Cylinder Height=200
    Cylinder Height Midpoint=1
    Cylinder Radius=40
    MaskSpawnGroup=false
    Offset=<70 100 20>
    Offset Mode=::NewEnumerator0
    Override Seed=false
    Rotation Angle=90
    Rotation Angle Type:ENiagara_AngleInput=::NewEnumerator0
    Rotation Axis=<0 1 0>
    Rotation Coordinate Space=::Simulation
    Rotation Mode=::NewEnumerator1
    Shape Primitive=::NewEnumerator1
    Surface Only=false
    UseOwnerScale=true
   )
   &InitialMeshOrientation(Use Rotation Vector=false)
   &AddVelocity(
    Cone Axis=<0 -1 0>
    Module.UseConeFalloff=false
    Module.UseVelDistribution=false
    Override Seed=false
    Velocity Mode=::NewEnumerator2
    Velocity Speed=&Multiply_Float(A=2, B=&RandomRangeFloat(Minimum=75, Maximum=35))
   )
   &AddRotationalVelocity(Rotation Rate=&Multiply_VectorByFloat(
     Vector=&RandomRangeVector(Minimum=<0.5 0.5 0.5>, Maximum=<1 1 1>)
     Float=&Multiply_Float(
      A=&UniformAOrBFloat(A=-1, Evaluation Type=::NewEnumerator0, Recalculate Random Each Loop=true, Override Seed=false)
      B=0.35
     )
    ))
  ]
  particleUpdateScript [
   &ParticleState()
   &ScaleMeshSize(Scale Factor=&VectorFromFloat(Value=&FloatFromCurve(
      FloatCurve=DataInterface() {propertyValues {Curve {keys [
          {arriveTangent=9.0725, interpMode=RCIM_Cubic, leaveTangent=9.0725, tangentMode=RCTM_User, time=0, value=0}
          {arriveTangent=0.3207, interpMode=RCIM_Cubic, leaveTangent=0.3207, tangentMode=RCTM_User, time=0.1408, value=1}
          {arriveTangent=-3.5468, interpMode=RCIM_Cubic, leaveTangent=-3.5468, tangentMode=RCTM_User, time=1, value=0}
         ]}}}
      Scale Curve=0.8
     )))
   &CurlNoiseForce(Pan Noise=false)
   &Drag(
    Drag=&Multiply_Float(
     A=&FloatFromCurve(FloatCurve=DataInterface() {propertyValues {Curve {keys [{interpMode=RCIM_Cubic, tangentMode=RCTM_User, time=0, value=0}, {interpMode=RCIM_Cubic, tangentMode=RCTM_User, time=1, value=2}]}}})
     B=&RandomRangeFloat(Minimum=0.8, Maximum=1.2)
    )
    Use Linear Drag=true
    Use Rotational Drag=false
   )
   &SolveForcesAndVelocity(Clamp Velocity=false, Limit Acceleration=false)
  ]
  rendererClasses [&NiagaraMeshRendererProperties]
  renderers [{rendererClass=&NiagaraMeshRendererProperties, rendererIndex=0}]
 }
 Straightwind {
  emitterUpdateScript [
   &EmitterState()
   &SpawnRate(Use Spawn Probability=false)
  ]
  particleSpawnScript [
   &InitializeParticle(
    Color=&FNTimeOfDayColor(
     Night Color=<rgba 0.4667 0.5464 0.5833 0.4>
     Morning Color=<rgba 0.5958 0.6976 0.7448 0.6>
     Day Color=<rgba 0.8 0.9367 1 0.6>
     Evening Color=<rgba 0.575 0.6732 0.7188 0.6>
    )
    Lifetime=&RandomRangeFloat(Minimum=0.8, Maximum=1.5)
    Mesh Scale Mode=::NewEnumerator3
    Mesh Uniform Scale=2
    UsePositionOffset=false
    Write Material Random=true
   )
   &InitialMeshOrientation(Mesh Axis To Orient=<0 1 0>, Mesh Orientation Mode=::NewEnumerator2, Use Rotation Vector=false)
   &ShapeLocation(
    Cylinder Height=50
    Cylinder Height Midpoint=1
    Cylinder Radius=40
    MaskSpawnGroup=false
    Offset=<80 400 50>
    Offset Mode=::NewEnumerator0
    Override Seed=false
    Shape Primitive=::NewEnumerator1
    Surface Only=false
    UseOwnerScale=true
   )
  ]
  particleUpdateScript [
   &ParticleState()
   &DynamicMaterialParameters(
    Index 0 Param 1=&FloatFromCurve(FloatCurve=DataInterface() {propertyValues {Curve {keys [
         {arriveTangent=2.4535, interpMode=RCIM_Cubic, leaveTangent=2.4535, tangentMode=RCTM_User, time=0, value=0}
         {interpMode=RCIM_Cubic, tangentMode=RCTM_User, time=0.5, value=0.5}
         {arriveTangent=-2.5865, tangentMode=RCTM_User, time=1, value=0}
        ]}}})
    Index 0 Param 3=&RandomRangeFloat(Minimum=0.8, Maximum=1.6)
    Index 1 Param 1=&RandomRangeFloat(Minimum=0, Maximum=1)
    Index 1 Param 2=&RandomRangeFloat(Minimum=0, Maximum=1)
    Module.Index0Param0Write=true
    Module.Index0Param1Write=true
    Module.Index0Param2Write=false
    Module.Index0Param3Write=false
    Module.Index1Param0Write=true
    Module.Index1Param1Write=true
    Module.Index1Param2Write=true
    Module.Index1Param3Write=true
    Write Parameter Index 1=true
   )
   &ScaleColor(
    Scale Alpha=&FloatFromCurve(
     FloatCurve=DataInterface() {propertyValues {Curve {keys [
         {interpMode=RCIM_Cubic, tangentMode=RCTM_User, time=0, value=0}
         {interpMode=RCIM_Cubic, tangentMode=RCTM_User, time=0.5, value=1}
         {tangentMode=RCTM_User, time=1, value=0}
        ]}}}
     Scale Curve=0.1
    )
    ScaleA=true
    ScaleRGB=true
   )
   /SetVariables(Particles.Position=&AddVectorToPosition(Position=$Particles.Position, Vector=&MakeVector(X=0, Y=-8, Z=0))) {isSetParameters=true, module=SetVariables_#1}
  ]
  rendererClasses [&NiagaraMeshRendererProperties]
  renderers [{rendererClass=&NiagaraMeshRendererProperties, rendererIndex=0}]
 }
 Straightwind001 {
  emitterUpdateScript [
   &EmitterState()
   &SpawnRate(SpawnRate=2, Use Spawn Probability=false)
  ]
  particleSpawnScript [
   &InitializeParticle(
    Color=&FNTimeOfDayColor(
     Night Color=<rgba 0.4667 0.5464 0.5833 0.4>
     Morning Color=<rgba 0.5958 0.6976 0.7448 0.6>
     Day Color=<rgba 0.8 0.9367 1 0.6>
     Evening Color=<rgba 0.575 0.6732 0.7188 0.6>
    )
    Lifetime=&RandomRangeFloat(Minimum=0.6, Maximum=1.5)
    Mesh Scale Mode=::NewEnumerator3
    Mesh Uniform Scale=2
    UsePositionOffset=false
    Write Material Random=true
   )
   &InitialMeshOrientation(Mesh Axis To Orient=<0 1 0>, Mesh Orientation Mode=::NewEnumerator2, Use Rotation Vector=false)
   &ShapeLocation(
    Cylinder Height=50
    Cylinder Height Midpoint=1
    Cylinder Radius=40
    MaskSpawnGroup=false
    Offset=<80 400 50>
    Offset Mode=::NewEnumerator0
    Override Seed=false
    Shape Primitive=::NewEnumerator1
    Surface Only=false
    UseOwnerScale=true
   )
  ]
  particleUpdateScript [
   &ParticleState()
   &DynamicMaterialParameters(
    Index 0 Param 1=&FloatFromCurve(FloatCurve=DataInterface() {propertyValues {Curve {keys [
         {arriveTangent=2.4535, interpMode=RCIM_Cubic, leaveTangent=2.4535, tangentMode=RCTM_User, time=0, value=0}
         {interpMode=RCIM_Cubic, tangentMode=RCTM_User, time=0.5, value=0.5}
         {arriveTangent=-2.5865, tangentMode=RCTM_User, time=1, value=0}
        ]}}})
    Index 0 Param 3=&RandomRangeFloat(Minimum=1.2, Maximum=1.6)
    Index 1 Param 1=&RandomRangeFloat(Minimum=0, Maximum=1)
    Index 1 Param 2=&RandomRangeFloat(Minimum=0, Maximum=1)
    Module.Index0Param0Write=true
    Module.Index0Param1Write=true
    Module.Index0Param2Write=true
    Module.Index0Param3Write=false
    Module.Index1Param0Write=true
    Module.Index1Param1Write=true
    Module.Index1Param2Write=true
    Module.Index1Param3Write=true
    Write Parameter Index 1=true
   )
   &ScaleColor(
    Scale Alpha=&FloatFromCurve(
     FloatCurve=DataInterface() {propertyValues {Curve {keys [
         {interpMode=RCIM_Cubic, tangentMode=RCTM_User, time=0, value=0}
         {interpMode=RCIM_Cubic, tangentMode=RCTM_User, time=0.5, value=1}
         {tangentMode=RCTM_User, time=1, value=0}
        ]}}}
     Scale Curve=0.1
    )
    ScaleA=true
    ScaleRGB=true
   )
   /SetVariables(Particles.Position=&AddVectorToPosition(Position=$Particles.Position, Vector=&MakeVector(X=0, Y=-8, Z=0))) {isSetParameters=true, module=SetVariables_#1}
  ]
  rendererClasses [&NiagaraMeshRendererProperties]
  renderers [{rendererClass=&NiagaraMeshRendererProperties, rendererIndex=0}]
 }
}
summary {
 systemName=NS_Emote_Stalemate_WindConstant
 emitters [
  {emitterName=Straightwind, rendererClasses [&NiagaraMeshRendererProperties]}
  {emitterName=Leaves_Vortex001, rendererClasses [&NiagaraMeshRendererProperties]}
  {emitterName=DirectionalBurst001, rendererClasses [&NiagaraSpriteRendererProperties]}
  {emitterName=Straightwind001, rendererClasses [&NiagaraMeshRendererProperties]}
 ]
 userVariables [
  User.bIsFrontEnd {type=NiagaraBool}
  User.bIsFrontEndPreview {type=NiagaraBool}
 ]
}
